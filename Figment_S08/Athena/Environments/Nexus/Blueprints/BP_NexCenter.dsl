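; BP_NexCenter is an actor that marks the Nexus center by publishing its world location to materials.
; It writes that location to the CenterLocation vector of the NexCollection material parameter collection.
; The construction script does this and prints the location as a debug string.
; BeginPlay repeats the write only where cosmetic events can execute; the overlap and tick events are empty.
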(event EventBeginPlay
  (Networking|CanExecuteCosmeticEvents
    (:True
      (Rendering|Material|SetVectorParameterValue "/Figment_S08/Environments/NexCollection.NexCollection" "CenterLocation" (Math|Conversions|ToLinearColor(Vector) (Transformation|GetActorLocation)) :!class /Script/Engine.KismetMaterialLibrary))))

(event Collision|EventActorBeginOverlap
  (params (OtherActor :type Actor%20Object%20Reference)))

(event EventTick
  (params (DeltaSeconds :type Float%20%28single-precision%29)))

(fn ConstructionScript
  (bind _returnvalue (Transformation|GetActorLocation))
  (Rendering|Material|SetVectorParameterValue "/Figment_S08/Environments/NexCollection.NexCollection" "CenterLocation" (Math|Conversions|ToLinearColor(Vector) _returnvalue) :!class /Script/Engine.KismetMaterialLibrary)
  (Development|PrintString (Utilities|String|ToString(Vector) _returnvalue)))
