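; Nexus Zero Point destabilize effect actor that plays a one-shot visual sequence on CE_TriggerEffect, then destroys itself.
; The trigger shows the post-process at full weight, zeroes the point light, then plays Timeline_0 from start.
; Each update drives MPC_Nexus spike, sphere-wave and post-process parameters, plus scaled light intensity.
; The sphere-wave tint writes to Spike_ColorOvrride, overwriting the spike color set earlier in that update.
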(event EventBeginPlay)

(event Custom|CE_TriggerEffect
  (bind _postprocess (Variables|Default|GetPostProcess))
  ;[1 0.7441 0] For debug
  (Utilities|FlowControl|Sequence
    (:then_0
      (Rendering|SetVisibility _postprocess true false)
      (Variables|/Script/Engine.PostProcessComponent|SetBlendWeight 1.0 _postprocess)
      (Rendering|Components|Light|SetIntensity (Variables|Default|GetPointLight) 0.0))
    (:then_1
      (bind (_direction _spike_scale _spike_motionstrength _spike_colorexp _pp_intensity _spherewave_timescale _spherewave_texturealpha _spherewave_masteralpha _light _pp_twitch _spike_colorovrride _spherewave_tint) (Timeline|Timeline_0 PlayFromStart)
        (:Update
          (Rendering|Material|SetScalarParameterValue "/Game/Athena/Environments/Nexus/Materials/FXshader/MPC_Nexus.MPC_Nexus" "Spike_Scale" _spike_scale :!class /Script/Engine.KismetMaterialLibrary)
          (Rendering|Material|SetScalarParameterValue "/Game/Athena/Environments/Nexus/Materials/FXshader/MPC_Nexus.MPC_Nexus" "Spike_MotionStrength" _spike_motionstrength :!class /Script/Engine.KismetMaterialLibrary)
          (Rendering|Material|SetScalarParameterValue "/Game/Athena/Environments/Nexus/Materials/FXshader/MPC_Nexus.MPC_Nexus" "Spike_ColorExp" _spike_colorexp :!class /Script/Engine.KismetMaterialLibrary)
          (Rendering|Material|SetVectorParameterValue "/Game/Athena/Environments/Nexus/Materials/FXshader/MPC_Nexus.MPC_Nexus" "Spike_ColorOvrride" _spike_colorovrride :!class /Script/Engine.KismetMaterialLibrary)
          (Rendering|Material|SetScalarParameterValue "/Game/Athena/Environments/Nexus/Materials/FXshader/MPC_Nexus.MPC_Nexus" "PP_Intensity" _pp_intensity :!class /Script/Engine.KismetMaterialLibrary)
          (Rendering|Material|SetScalarParameterValue "/Game/Athena/Environments/Nexus/Materials/FXshader/MPC_Nexus.MPC_Nexus" "SphereWave_TimeScale" _spherewave_timescale :!class /Script/Engine.KismetMaterialLibrary)
          (Rendering|Material|SetScalarParameterValue "/Game/Athena/Environments/Nexus/Materials/FXshader/MPC_Nexus.MPC_Nexus" "SphereWave_TextureAlpha" _spherewave_texturealpha :!class /Script/Engine.KismetMaterialLibrary)
          (Rendering|Material|SetVectorParameterValue "/Game/Athena/Environments/Nexus/Materials/FXshader/MPC_Nexus.MPC_Nexus" "Spike_ColorOvrride" _spherewave_tint :!class /Script/Engine.KismetMaterialLibrary)
          (Rendering|Material|SetScalarParameterValue "/Game/Athena/Environments/Nexus/Materials/FXshader/MPC_Nexus.MPC_Nexus" "SphereWave_MasterAlpha" _spherewave_masteralpha :!class /Script/Engine.KismetMaterialLibrary)
          (Rendering|Components|Light|SetIntensity (Variables|Default|GetPointLight) (* _light 4e+07))
          (Rendering|Material|SetScalarParameterValue "/Game/Athena/Environments/Nexus/Materials/FXshader/MPC_Nexus.MPC_Nexus" "PP_Twitch" _pp_twitch :!class /Script/Engine.KismetMaterialLibrary))
        (:Finished
          ;[1 0 0] Use this when in a real game
          (Actor|DestroyActor))))))

(fn ConstructionScript)
