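# Unlit translucent two-sided aura material for the Nexus butterfly orbs, drawing wispy smoke on spline-thickened geometry.
# A panned smoke texture blends two colors and drives opacity, faded at UV edges, depth intersections, near the camera and by view angle.
# World position offset sums noise-modulated SplineThicken, an end-pinned two-sine wiggle and an MPC_Cube MoveOrbs shift toward a hard-coded world point.
# That offset passes through MF_SpawnInBetween.
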Material /Figment_S08/Athena/Environments/Nexus/Materials/M_Butterfly_Orbs_Aura
graph {
 properties {
  bUsedWithMeshParticles=true
  bUsedWithParticleSprites=true
  blendMode=BLEND_Translucent
  pixelDepthOffsetMode=PDOM_Legacy
  shadingModel=MSM_Unlit
  twoSided=true
  usedShadingModels=Unlit
 }
 nodes [
  n1 = TextureSample!(
   UVs=Panner(Coordinate=TextureCoordinate() {uTiling=0.5, vTiling=2}) {speedX=0.025, speedY=0.1}
   Tex
   Level=Constant() {r=4}
   Apply View MipBias
  ) {mipValueMode=TMVM_MipLevel, texture=/Game/Effects/Fort_Effects/Textures/Tile/T_Inky_Smoke_Tile}
  LinearInterpolate(
   A=VectorParameter() {parameterName=Color1, defaultValue=<rgba 0.0163 0.013 0.25 1>}.RGB
   B=VectorParameter() {parameterName=Color2, defaultValue=<rgba 0.0937 0.1719 0.25 1>}.RGB
   Alpha=@n1.G
  )
  n2 = TextureCoordinate() {uTiling=0.25, vTiling=2}
  n3 = MaterialFunctionCall(
   WidthBase=ScalarParameter() {defaultValue=1000, parameterName="WidthBase (S)"}
   WidthTip=ScalarParameter() {defaultValue=5000, parameterName="WidthTip (S)"}
   UVs for Thickness=TextureSample!(UVs=Panner(Coordinate=@n2) {speedX=0.0015, speedY=0.025}, Tex, Apply View MipBias) {texture=/Game/Effects/Fort_Effects/Gameplay/SmokeDustExplosions/Textures/T_NoisePack_01}.G
  ) {materialFunction=&SplineThicken}
  n4 = Multiply(
   A=ComponentMask(None=TextureCoordinate()) {g=true}
   B=MaterialFunctionCall() {materialFunction=/Engine/Functions/Engine_MaterialFunctions02/WorldPositionOffset/ObjectScale}.Scale X
  )
  n5 = Multiply(A=Time(), B=ScalarParameter() {defaultValue=0.25, parameterName=WiggleSpeed})
  n6 = Multiply(A=@n5) {constB=2}
  n7 = Multiply(
   A=Sine(None=Add(A=@n4, B=@n6)) {period=7}
   B=Transform(None=Constant3Vector() {constant=<rgba 0 0 3 0>}) {transformSourceType=TRANSFORMSOURCE_Local}
  )
  n8 = Multiply(
   A=ComponentMask(None=TextureCoordinate()) {g=true}
   B=MaterialFunctionCall() {materialFunction=&ObjectScale}.Scale X
  )
  n9 = Multiply(A=@n5)
  n10 = Multiply(
   A=Sine(None=Add(A=@n8, B=@n9)) {period=5}
   B=Transform(None=Constant3Vector() {constant=<rgba 0 1.5 1.5 0>}) {transformSourceType=TRANSFORMSOURCE_Local}
  )
  n11 = MaterialFunctionCall() {materialFunction=&LinearGradient}
  n12 = OneMinus(None=@n11.VGradient)
  n13 = Multiply(
   A=Multiply(A=Add(A=@n7, B=@n10), B=ScalarParameter() {defaultValue=20, parameterName=WiggleAmount})
   B=Multiply(A=Multiply(A=@n11.VGradient, B=@n12)) {constB=4}
  )
  n14 = Normalize(VectorInput=Subtract(A=Constant3Vector() {constant=<rgba -713 2456 244652 0>}, B=ObjectPositionWS()))
  n15 = CollectionParameter() {collection=/Game/Athena/Prototype/Blueprints/Cube/Materials/MPC_Cube, parameterName=MoveOrbs}
  MaterialFunctionCall!(WPO=Add(A=Add(A=@n3.WorldPositionOffset, B=@n13), B=Multiply(A=@n14, B=@n15))) {materialFunction=/Game/Athena/Environments/Nexus/Materials/MF_SpawnInBetween}
  n16 = ScalarParameter() {defaultValue=2, parameterName=AlphaBoost}
  n17 = MaterialFunctionCall() {materialFunction=&LinearGradient}
  n18 = Multiply(A=Multiply(A=@n17.UGradient, B=OneMinus(None=@n17.UGradient))) {constB=4}
  n19 = Multiply(A=Multiply(A=@n17.VGradient, B=OneMinus(None=@n17.VGradient))) {constB=4}
  n20 = DepthFade(Opacity=Multiply(A=@n1.G, B=Multiply(A=@n18, B=@n19)), FadeDistance=ScalarParameter() {defaultValue=500, parameterName=FadeDistance})
  n21 = Constant() {r=2048}
  n22 = Constant() {r=2048}
  Multiply(
   A=Multiply(
    A=Multiply(A=@n16, B=@n20)
    B=MaterialFunctionCall(Fade Length=@n21, Fade Offset=@n22) {materialFunction=&CameraDepthFade}.Result
   )
   B=MaterialFunctionCall("FadeSensitivity (optional)"=ScalarParameter() {defaultValue=1, parameterName=AngleFadeAmount}) {materialFunction=&FadeBasedOnViewAngle}.Result
  )
 ]
}
