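# Unlit kaleidoscope material for the Season 8 Nexus environment, drawing shimmering chromatic mirror surfaces with glowing tunnel and sky-sphere variants.
# Two panning, flattened mirror normals, one bump-offset, bend a reflection vector that samples chromatic smoke directly and through Fresnel, tinted by the mirror texture.
# View-angle and UV core glows are added, plus time-animated sine tunnel bands when PushToSphere is off.
# Static switches compute camera-following EndlessTunnel parallax and sphere-projection offsets.
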Material /Figment_S08/Athena/Environments/Nexus/Materials/M_Kaleidoscop
graph {
 properties {pixelDepthOffsetMode=PDOM_Legacy, shadingModel=MSM_Unlit, usedShadingModels=Unlit}
 nodes [
  CameraVectorWS()
  n1 = ComponentMask(None=WorldPosition().XYZ) {b=true}
  n2 = ScalarParameter() {parameterName=ZeroPoint}
  Clamp(None=Divide(A=Subtract(A=@n1, B=@n2), B=ScalarParameter() {defaultValue=28087.7734, parameterName=FadeDistance}))
  Constant()
  Constant() {r=1}
  n3 = VectorParameter() {parameterName=Color_Background_2, defaultValue=<rgba 2.1963 4 3.8006 1>}
  n4 = Subtract(A=CameraPositionWS(), B=WorldPosition().XYZ)
  n5 = Clamp(None=DotProduct(A=Constant3Vector() {constant=<rgba 0 0 -1 0>}, B=Normalize(VectorInput=@n4)))
  n6 = OneMinus(None=@n5)
  n7 = Constant() {r=16}
  n8 = QualitySwitch(Default=ScalarParameter() {defaultValue=14, parameterName=Contrast_SoulsFade_Exp_3}, Low=@n7, Medium=@n7)
  n9 = Multiply(A=@n3.RGB, B=Clamp(None=Power(Base=@n6, Exp=@n8)))
  n10 = ScalarParameter() {defaultValue=2, parameterName=Tiling_SoulsLights_x_1}
  n11 = ScalarParameter() {defaultValue=2, parameterName=Tiling_SoulsLights_y_1}
  n12 = Panner(
   Coordinate=Multiply(A=TextureCoordinate() {coordinateIndex=2}, B=AppendVector(A=@n10, B=@n11))
   Speed=AppendVector(A=ScalarParameter() {defaultValue=0.01, parameterName=SpeedX_4}, B=ScalarParameter() {defaultValue=0.005, parameterName=SpeedY_4})
  ) {bFractionalPart=true}
  n13 = TextureSample(UVs=@n12) {texture=/Figment_S08/Athena/Environments/Nexus/Textures/KaleidosMirror_D}
  n14 = MaterialFunctionCall(
   Normal=TextureSample(UVs=@n12) {samplerType=SAMPLERTYPE_Normal, texture=/Game/Athena/Environments/Nexus/Textures/KaleidosMirror_N}.RGB
   Flatness=ScalarParameter() {defaultValue=0.66, parameterName=NormalIntensity}
  ) {materialFunction=&FlattenNormal}
  n15 = TextureCoordinate() {coordinateIndex=2}
  n16 = AppendVector(
   A=ScalarParameter() {defaultValue=2, parameterName=Tiling_SoulsLights_x_2}
   B=ScalarParameter() {defaultValue=2, parameterName=Tiling_SoulsLights_y_2}
  )
  n17 = Panner(
   Coordinate=BumpOffset(Coordinate=Multiply(A=@n15, B=@n16), Height=Constant() {r=1}, HeightRatioInput=ScalarParameter() {defaultValue=10, parameterName=Offset})
   Speed=AppendVector(A=ScalarParameter() {defaultValue=0.01, parameterName=SpeedX_1}, B=ScalarParameter() {defaultValue=0.005, parameterName=SpeedY_1})
  ) {bFractionalPart=true}
  n18 = Multiply(
   A=Constant3Vector() {constant=<rgba 1 1 0 0>}
   B=MaterialFunctionCall(
    Normal=TextureSample(UVs=@n17) {samplerType=SAMPLERTYPE_Normal, texture=&KaleidosMirror_N}.RGB
    Flatness=ScalarParameter() {defaultValue=0.9238, parameterName=NormalIntensity2}
   ) {materialFunction=&FlattenNormal}.Result
  )
  n19 = ReflectionVectorWS(CustomWorldNormal=Transform(None=Add(A=@n14.Result, B=@n18)))
  n20 = OneMinus(None=Fresnel(ExponentIn=ScalarParameter() {defaultValue=1, parameterName=CromaticExponent}, Normal=@n19) {baseReflectFraction=0})
  n21 = Multiply(
   A=ScalarParameter() {defaultValue=1.2, parameterName=Brightness}
   B=TextureSample(UVs=ComponentMask(None=@n19) {g=true, r=true}) {texture=/Game/Effects/Fort_Effects/Textures/Tile/T_Inky_Smoke_Tile_Chromatic}.RGB
  )
  n22 = Multiply(
   A=LinearInterpolate(
    A=VectorParameter() {parameterName=TintPattern1, defaultValue=<rgba 0.1615 0.1615 0.1615 1>}.RGB
    B=VectorParameter() {parameterName=TintPattern2, defaultValue=<rgba 1 1 1 1>}.RGB
    Alpha=@n13.R
   )
   B=Add(A=TextureSample(UVs=@n20) {texture=&T_Inky_Smoke_Tile_Chromatic}.RGB, B=@n21)
  )
  n23 = VectorParameter() {parameterName=ColorZeroPoint, defaultValue=<rgba 3.6551 1.8123 5 1>}
  n24 = Power(Base=@n5, Exp=ScalarParameter() {defaultValue=5.3107, parameterName=Contrast_SoulsFade_Exp})
  n25 = Power(Base=ComponentMask(None=TextureCoordinate()) {g=true}, Exp=ScalarParameter() {defaultValue=50, parameterName=Contrast_Centercore})
  n26 = Add(
   A=Add(A=Add(A=@n9, B=@n22), B=Multiply(A=@n23.RGB, B=@n24))
   B=Multiply(A=VectorParameter() {parameterName=ColorCenterCore, defaultValue=<rgba 3.6551 1.8123 5 1>}.RGB, B=Clamp(None=@n25))
  )
  n27 = VectorParameter() {parameterName=CoreTunnel, defaultValue=<rgba 0.8385 0.8385 0.8385 1>}
  n28 = ComponentMask(None=TextureCoordinate()) {g=true}
  n29 = ScalarParameter() {defaultValue=8, parameterName=Contrast_Tunnel_Power}
  n30 = Time()
  n31 = ScalarParameter() {defaultValue=0.3238, parameterName=SpeedSine}
  n32 = Add(
   A=Multiply(A=Power(Base=@n28, Exp=@n29), B=ScalarParameter() {defaultValue=-17.4276, parameterName=Contrast_Tunnel_Sine})
   B=Frac(None=Multiply(A=@n30, B=@n31))
  )
  n33 = ScalarParameter() {defaultValue=-2.6943, parameterName=Contrast_Tunnel}
  n34 = ScalarParameter() {defaultValue=1.5865, parameterName=Contrast_Tunnel2}
  n35 = Multiply(A=Saturate(None=Sine(None=@n32)), B=Saturate(None=LinearInterpolate(A=@n33, B=@n34, Alpha=@n28)))
  LinearInterpolate(A=@n26, B=@n27.RGB, Alpha=@n35)
  n36 = ComponentMask(None=TextureCoordinate() {coordinateIndex=1}) {g=true}
  MaterialFunctionCall(
   Transition Phase=Multiply(A=OneMinus(None=@n36), B=ScalarParameter() {defaultValue=1.2836, parameterName=Transition})
   Height Texture=@n13.R
   Contrast=ScalarParameter() {defaultValue=1, parameterName=ContrastTran}
  ) {materialFunction=&HeightLerp}
  MaterialFunctionCall() {materialFunction=&CameraDirectionVector}
  Multiply(A=@n21, B=VectorParameter() {parameterName=Color_Background_1, defaultValue=<rgba 0.5588 0.1824 0.974 1>}.RGB)
  n37 = ScalarParameter() {defaultValue=0.9, parameterName=LowMedBrightnessMult}
  Multiply(A=QualitySwitch(Default=Constant() {r=1}, Low=@n37, Medium=@n37), B=@n3.RGB)
  OneMinus()
  OneMinus()
  n38 = CameraPositionWS()
  n39 = ScalarParameter() {defaultValue=1, parameterName=Paralaxing}
  n40 = TextureCoordinate()
  n41 = StaticSwitchParameter(True=Multiply(A=Multiply(A=@n38, B=@n39), B=ComponentMask(None=@n40) {g=true}), False=Constant3Vector()) {parameterName=EndlessTunnel}
  n42 = Add(A=WorldPosition().XYZ, B=@n41)
  n43 = Normalize(VectorInput=@n42)
  n44 = Subtract(A=ScalarParameter() {defaultValue=127235.3984, parameterName=Radius Of Sky}, B=Distance(A=@n42, B=Constant3Vector()))
  StaticSwitchParameter(True=Add(A=Multiply(A=@n43, B=@n44), B=@n41), False=@n41) {parameterName=PushToSphere}
  n45 = Multiply(
   A=Power(Base=@n28, Exp=ScalarParameter() {defaultValue=15.2806, parameterName=Contrast_Tunnel_Power2})
   B=ScalarParameter() {defaultValue=50, parameterName=Contrast_Tunnel_PowerBrightness2}
  )
  n46 = Multiply(A=@n27.RGB, B=@n35)
  StaticSwitchParameter(True=@n26, False=Add(A=Multiply(A=@n45, B=@n27.RGB), B=Add(A=@n46, B=@n26))) {parameterName=PushToSphere}
  TextureSample(UVs=@n17) {texture=&KaleidosMirror_D}
 ]
}
