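# M_NexBlendRing is an unlit, translucent Nanite-flagged material in the Nexus environment set, apparently for a soft-blended ring.
# It raises the red vertex color channel to a Falloff parameter, so the gradient comes from mesh vertex painting rather than textures.
# A separate Color parameter supplies an over-bright pale cyan tint, and translucent vertex fog is turned off.
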Material /Figment_S08/Athena/Environments/Nexus/Materials/M_NexBlendRing
graph {
 properties {
  bUseTranslucencyVertexFog=false
  bUsedWithNanite=true
  blendMode=BLEND_Translucent
  pixelDepthOffsetMode=PDOM_Legacy
  shadingModel=MSM_Unlit
  usedShadingModels=Unlit
 }
 nodes [
  Power(Base=VertexColor().R, Exp=ScalarParameter() {defaultValue=1.6819, parameterName=Falloff})
  VectorParameter() {parameterName=Color, defaultValue=<rgba 1.7448 2.5 2.5 1>}
 ]
}
