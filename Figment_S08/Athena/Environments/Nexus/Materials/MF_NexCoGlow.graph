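# Material function that gives a Nexus surface a glowing, swirling storm core look, overriding Normal, Emissive, Roughness and Specular on incoming attributes.
# A fake spherical normal comes from world position relative to SphereCenter, pushed along the camera-to-center direction.
# Storm swirl textures on two camera-facing virtual planes pan oppositely, height-blended by view angle to distort that normal.
# Emissive sums fresnel gradient, core glow, rim band and top-down tint; Specular blends by distance from center.
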MaterialFunction /Figment_S08/Athena/Environments/Nexus/Materials/MF_NexCoGlow
graph {nodes [
  n1 = FunctionInput() {inputType=FunctionInput_MaterialAttributes}
  n2 = WorldPosition()
  Add(A=BreakMaterialAttributes!(Attr=@n1).WorldPositionOffset, B=@n2.XYZ)
  n3 = VectorParameter() {parameterName=SphereCenter, defaultValue=<rgba -3024 -1776 96 1>}
  n4 = Normalize(VectorInput=Subtract(A=@n3.RGB, B=CameraPositionWS()))
  n5 = Multiply(A=@n4, B=ScalarParameter() {defaultValue=9151.5371, parameterName=OffsetCenter})
  n6 = Normalize(VectorInput=Subtract(A=@n2.XYZ, B=Add(A=@n3.RGB, B=@n5)))
  n7 = CameraVectorWS()
  n8 = CrossProduct(A=@n7, B=VectorParameter() {parameterName=Axis, defaultValue=<rgba 0 0 1 1>}.RGB)
  n9 = Panner(
   Coordinate=MaterialFunctionCall(
    Plane Normal Axis=@n7
    Plane X Axis=Normalize(VectorInput=@n8)
    Plane Center=@n3.RGB
    UV Scale=ScalarParameter() {defaultValue=1995.9475, parameterName=SizeProjectionTexture}
   ) {materialFunction=&VirtualPlaneCoordinates}.UVs
   Speed=ScalarParameter() {defaultValue=-0.0667, parameterName=PannerSpeed1}
  )
  n10 = CameraVectorWS()
  n11 = CrossProduct(A=@n10, B=VectorParameter() {parameterName=Axis, defaultValue=<rgba 0 0 1 1>}.RGB)
  n12 = Multiply(A=@n4, B=ScalarParameter() {defaultValue=3086.6487, parameterName=OffsetCenter2})
  n13 = Panner(
   Coordinate=MaterialFunctionCall(
    Plane Normal Axis=@n10
    Plane X Axis=Normalize(VectorInput=@n11)
    Plane Center=Add(A=@n3.RGB, B=@n12)
    UV Scale=ScalarParameter() {defaultValue=4321.3545, parameterName=SizeProjectionTexture2}
   ) {materialFunction=&VirtualPlaneCoordinates}.UVs
   Speed=ScalarParameter() {defaultValue=0.0667, parameterName=PannerSpeed2}
  )
  n14 = DotProduct(A=CameraVectorWS(), B=@n6)
  n15 = ScalarParameter() {defaultValue=6.1705, parameterName=FresnelEXPTreansition}
  n16 = TextureSample(UVs=@n9) {samplerType=SAMPLERTYPE_LinearGrayscale, texture=/Figment_S08/VisualThreatManager/StormVisuals/LargeClouds/T_StormSwirlMask}
  n17 = MaterialFunctionCall(
   Transition Phase=Clamp(None=Power(Base=@n14, Exp=@n15))
   Height Texture=OneMinus(None=@n16.R)
   Contrast=ScalarParameter() {defaultValue=1, parameterName=Contrast}
  ) {materialFunction=&HeightLerp}
  n18 = LinearInterpolate(
   A=TextureSample(UVs=@n9) {samplerType=SAMPLERTYPE_Normal, texture=/Game/VisualThreatManager/StormVisuals/LargeClouds/T_StormSwirlNormal}.RGB
   B=TextureSample(UVs=@n13) {samplerType=SAMPLERTYPE_Normal, texture=&T_StormSwirlNormal}.RGB
   Alpha=@n17.Alpha
  )
  n19 = Power(
   Base=DotProduct(
    A=MaterialFunctionCall(BaseNormal=@n6, AdditionalNormal=@n18) {materialFunction=&BlendAngleCorrectedNormals}.Result
    B=Constant3Vector() {constant=<rgba 0 0 1 0>}
   )
   Exp=ScalarParameter() {defaultValue=3.0279, parameterName=TopDownFresnell}
  )
  n20 = VectorParameter() {parameterName=EmissiveTopDown, defaultValue=<rgba 0.0158 0.0078 0.0312 1>}
  n21 = ScalarParameter() {defaultValue=0.0884, parameterName=FresnelEXP}
  n22 = VectorParameter() {parameterName=Stretchdistortion, defaultValue=<rgba 1.1 1.1 1 1>}
  n23 = LinearInterpolate(
   A=@n16.R
   B=TextureSample(UVs=@n13) {samplerType=SAMPLERTYPE_LinearGrayscale, texture=&T_StormSwirlMask}.R
   Alpha=@n17.Alpha
  )
  n24 = ScalarParameter() {defaultValue=0.3029, parameterName=Distortion}
  n25 = Multiply(A=LinearInterpolate(A=Constant3Vector() {constant=<rgba 1 1 1 0>}, B=@n22.RGB, Alpha=Multiply(A=@n23, B=@n24)), B=@n6)
  n26 = ScalarParameter() {defaultValue=88.6155, parameterName=FresnelEXP3}
  n27 = Add(
   A=LinearInterpolate(
    A=VectorParameter() {parameterName=EmColor1, defaultValue=<rgba 0.2727 0.133 0.3438 1>}.RGB
    B=VectorParameter() {parameterName=EmColor2, defaultValue=<rgba 0 0 0 1>}.RGB
    Alpha=Fresnel(ExponentIn=@n21, Normal=@n25) {baseReflectFraction=0}
   )
   B=Multiply(A=VectorParameter() {parameterName=CoreGLow, defaultValue=<rgba 0.4069 0.6973 1 1>}.RGB, B=Power(Base=@n14, Exp=@n26))
  )
  n28 = DotProduct(A=CameraVectorWS(), B=@n25)
  n29 = Multiply(
   A=LinearInterpolate(A=ScalarParameter() {defaultValue=-4.1348, parameterName=C1}, B=ScalarParameter() {defaultValue=1.2052, parameterName=C2}, Alpha=@n28)
   B=LinearInterpolate(A=ScalarParameter() {defaultValue=2.901, parameterName=C3}, B=ScalarParameter() {defaultValue=-0.223, parameterName=C4}, Alpha=@n28)
  )
  n30 = Multiply(
   A=Power(Base=Clamp(None=@n29), Exp=ScalarParameter() {defaultValue=1.3611, parameterName=FresnelEXP4})
   B=VectorParameter() {parameterName=CoreRim, defaultValue=<rgba 0.1445 0.5088 0.5781 1>}.RGB
  )
  n31 = WorldPosition()
  n32 = Divide(
   A=Subtract(A=Distance(A=@n31.XYZ, B=@n3.RGB), B=ScalarParameter() {defaultValue=1452.7511, parameterName=Offsetdistance})
   B=ScalarParameter() {defaultValue=95.0929, parameterName=Blenddistance}
  )
  n33 = SetMaterialAttributes!(
   MaterialAttributes=@n1
   Normal=@n6
   EmissiveColor=Add(A=Multiply(A=@n19, B=@n20.RGB), B=Add(A=@n27, B=@n30))
   Roughness=ScalarParameter() {parameterName=Roughness}
   Specular=LinearInterpolate(A=ScalarParameter() {defaultValue=0.25, parameterName=Specular}, B=ScalarParameter() {parameterName=Specular2}, Alpha=Clamp(None=@n32))
  ) {attributeSetTypes [
    #1
    #2
    #3
    #4
   ]}
  FunctionOutput(None=@n33)
  Multiply(A=@n23, B=@n22.RGB)
  Multiply(A=@n6) {constB=-1}
  n34 = ScalarParameter() {defaultValue=0.01, parameterName=FresnelEXP2}
  Multiply(A=OneMinus(None=Fresnel(ExponentIn=@n34, Normal=@n6)), B=@n26)
  Power(Base=OneMinus(None=@n28), Exp=@n21)
  Transform(None=@n25) {transformSourceType=TRANSFORMSOURCE_World, transformType=TRANSFORM_Tangent}
  VectorParameter() {parameterName=Axis, defaultValue=<rgba 0 0 1 1>}
  VectorParameter() {parameterName=Axis, defaultValue=<rgba 0 0 1 1>}
 ]}
