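# Nanite-enabled opaque Default Lit material M_NexCo for the Figment Nexus environment, apparently a glowing core surface.
# Color, Specular and Roughness parameters build material attributes that pass through MF_NexCoGlow.
# A separate emissive branch adds a Fresnel color blend to a blue core, made from axis-scaled vertex normals dotted with the camera vector and raised to a power.
# That branch's sum feeds nothing, leaving EmissiveColor unconnected before the function.
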Material /Figment_S08/Athena/Environments/Nexus/Materials/M_NexCo
graph {
 properties {
  bTangentSpaceNormal=false
  bUseMaterialAttributes=true
  bUsedWithNanite=true
  pixelDepthOffsetMode=PDOM_Legacy
  usedShadingModels=Default Lit
 }
 nodes [
  n1 = ScalarParameter() {defaultValue=3.2606, parameterName=Exp}
  n2 = DotProduct(A=Multiply(A=VertexNormalWS(), B=VectorParameter() {parameterName=SquezeNormals, defaultValue=<rgba 1 1 11.3948 1>}.RGB), B=CameraVectorWS())
  n3 = ScalarParameter() {defaultValue=3.3611, parameterName=ExpBlueCore}
  Add(
   A=LinearInterpolate(
    A=VectorParameter() {parameterName=EmissiveFresnel1, defaultValue=<rgba 0 0 0 1>}.RGB
    B=VectorParameter() {parameterName=EmissiveFresnel2, defaultValue=<rgba 0.1123 0.0091 0.25 1>}.RGB
    Alpha=Fresnel(ExponentIn=@n1)
   )
   B=Multiply(A=Power(Base=@n2, Exp=@n3), B=VectorParameter() {parameterName=EmissvieBlueCore, defaultValue=<rgba 0.1136 0.2452 0.401 1>}.RGB)
  )
  MaterialFunctionCall!(In=MakeMaterialAttributes!(
    BaseColor=VectorParameter() {parameterName=Color, defaultValue=<rgba 0 0 0 1>}.RGB
    Metallic
    Specular=ScalarParameter() {parameterName=Specular}
    Roughness=ScalarParameter() {parameterName=Roughness}
    Anisotropy
    EmissiveColor
    Opacity
    OpacityMask
    Normal
    Tangent
    WorldPositionOffset
    SubsurfaceColor
    ClearCoat
    ClearCoatRoughness
    AmbientOcclusion
    Refraction
    CustomizedUVs_0
    CustomizedUVs_1
    CustomizedUVs_2
    CustomizedUVs_3
    CustomizedUVs_4
    CustomizedUVs_5
    CustomizedUVs_6
    CustomizedUVs_7
    PixelDepthOffset
    ShadingModel
    Displacement
   )) {materialFunction=&/MF_NexCoGlow}
 ]
}
