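# Unlit translucent material for the NexCo carpet in the Nexus environment, drawing a soft procedural gradient without textures.
# A U*(1-U) falloff shaped by a power and multiplied by inverted V forms the main mask.
# The mask and a sharp V-based core blend three colors, while an opacity term fades out inside that core.
# A normalized world-to-camera vector scaled by Offset provides a small shift toward the camera.
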Material /Figment_S08/Athena/Environments/Nexus/Materials/M_NexCoCarpet
graph {
 properties {blendMode=BLEND_Translucent, pixelDepthOffsetMode=PDOM_Legacy, shadingModel=MSM_Unlit, usedShadingModels=Unlit}
 nodes [
  n1 = TextureCoordinate()
  n2 = ComponentMask(None=@n1) {r=true}
  n3 = Multiply(A=@n2, B=OneMinus(None=@n2))
  n4 = ScalarParameter() {defaultValue=0.7315, parameterName=Power}
  n5 = OneMinus(None=ComponentMask(None=@n1) {g=true})
  n6 = Multiply(A=Multiply(A=Power(Base=@n3, Exp=@n4), B=ScalarParameter() {defaultValue=3.2981, parameterName=Opacity}), B=@n5)
  n7 = ScalarParameter() {defaultValue=21.027, parameterName=Power2}
  n8 = ScalarParameter() {defaultValue=18.0122, parameterName=Power3}
  n9 = Saturate(None=Multiply(A=Power(Base=@n5, Exp=@n8), B=ScalarParameter() {defaultValue=3.8193, parameterName=CoreBlend}))
  LinearInterpolate(
   A=LinearInterpolate(
    A=VectorParameter() {parameterName=Color, defaultValue=<rgba 0.0552 0.0382 0.1354 1>}.RGB
    B=VectorParameter() {parameterName=emissive, defaultValue=<rgba 0.1326 0.2595 0.3021 1>}.RGB
    Alpha=Power(Base=@n6, Exp=@n7)
   )
   B=VectorParameter() {parameterName=emissive2, defaultValue=<rgba 0.0537 0.0876 0.151 1>}.RGB
   Alpha=@n9
  )
  Multiply(A=Multiply(A=OneMinus(None=@n9), B=@n6), B=ScalarParameter() {defaultValue=0.74, parameterName=Opacity2})
  n10 = CameraPositionWS()
  n11 = WorldPosition()
  Multiply(A=Normalize(VectorInput=Subtract(A=@n10, B=@n11.XYZ)), B=ScalarParameter() {defaultValue=3, parameterName=Offset})
 ]
}
