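# Translucent two-sided material for animated ink strands in the Nexus environment, with per-actor time offsets from actor position.
# Spline-thickened geometry moves with two rotated sine waves and is pushed outside a sphere around the actor.
# Two distorted panning ink samples drive color and end-faded, depth-faded opacity.
# Their normals blend with a camera-biased spherical normal around the NexCollection center, which also anchors a core glow; mobile has separate colors.
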Material /Figment_S08/Athena/Environments/Nexus/Materials/M_NexCoInk
graph {
 properties {
  bAlwaysEvaluateWorldPositionOffset=true
  bTangentSpaceNormal=false
  bUseTranslucencyVertexFog=false
  blendMode=BLEND_Translucent
  pixelDepthOffsetMode=PDOM_Legacy
  twoSided=true
  usedShadingModels=Default Lit
 }
 nodes [
  n1 = Add(A=NamedRerouteUsage(), B=WorldPosition().XYZ)
  n2 = ComponentMask(None=CollectionParameter() {collection=/Figment_S08/Environments/NexCollection, parameterName=CenterLocation}) {b=true, g=true, r=true}
  n3 = Subtract(A=@n2, B=CameraPositionWS())
  n4 = Add(A=@n2, B=Multiply(A=Normalize(VectorInput=@n3), B=ScalarParameter() {defaultValue=5000, parameterName=NormalCenteroffest}))
  n5 = ScalarParameter() {defaultValue=1, parameterName=TilingX}
  n6 = ScalarParameter() {defaultValue=0.3767, parameterName=TilingY}
  n7 = Time()
  n8 = ComponentMask(None=ActorPositionWS()) {r=true}
  n9 = Add(A=@n8, B=@n8)
  n10 = ScalarParameter() {defaultValue=124.3456, parameterName=Random}
  n11 = Multiply(A=Frac(None=Divide(A=@n9, B=@n10)), B=ScalarParameter() {defaultValue=10.1778, parameterName=RandomOffset})
  n12 = Add(A=@n7, B=@n11)
  n13 = Panner(
   Coordinate=Multiply(A=TextureCoordinate(), B=AppendVector(A=@n5, B=@n6))
   Time=@n12
   Speed=AppendVector(A=ScalarParameter() {parameterName=SpeedX}, B=ScalarParameter() {defaultValue=0.05, parameterName=SpeedY})
  )
  n14 = Multiply(
   A=TextureCoordinate()
   B=AppendVector(A=ScalarParameter() {defaultValue=0.2571, parameterName=NoiseTiling}, B=ScalarParameter() {defaultValue=1.3192, parameterName=NoiseTilingY})
  )
  n15 = Multiply(
   A=ComponentMask(None=TextureSample(UVs=@n14) {texture=/Game/Effects/Fort_Effects/Gameplay/SmokeDustExplosions/Textures/T_NoisePack_01}.RGB) {g=true, r=true}
   B=ScalarParameter() {defaultValue=0.05, parameterName=Distort}
  )
  n16 = Add(A=@n13, B=@n15)
  n17 = TextureSample(UVs=@n16) {samplerType=SAMPLERTYPE_Normal, texture=/Figment_S08/Athena/Environments/Nexus/Textures/T_NexCoInk_N}
  n18 = Add(A=Add(A=@n13, B=Constant2Vector() {g=0.5}), B=@n15)
  n19 = TextureSample(UVs=@n18) {samplerType=SAMPLERTYPE_Normal, texture=&T_NexCoInk_N}
  n20 = MaterialFunctionCall(
   BaseNormal=Normalize(VectorInput=Subtract(A=@n1, B=@n4))
   AdditionalNormal=MaterialFunctionCall(
    Normal=MaterialFunctionCall(BaseNormal=@n17.RGB, AdditionalNormal=@n19.RGB) {materialFunction=&BlendAngleCorrectedNormals}.Result
    Flatness=ScalarParameter() {defaultValue=0.0095, parameterName=FlattenNormal}
   ) {materialFunction=&FlattenNormal}.Result
  ) {materialFunction=&BlendAngleCorrectedNormals}
  n21 = Fresnel(ExponentIn=ScalarParameter() {defaultValue=1, parameterName=CromaticExponent}, Normal=@n20.Result) {baseReflectFraction=0}
  n22 = OneMinus(None=@n21)
  Add(A=Multiply(A=TextureSample(UVs=@n22) {texture=/Game/Effects/Fort_Effects/Textures/Tile/T_Inky_Smoke_Tile_Chromatic}.RGB, B=@n21))
  n23 = ComponentMask(None=TextureCoordinate()) {g=true}
  n24 = LinearInterpolate(
   A=ScalarParameter() {defaultValue=4243.7358, parameterName=Width}
   B=ScalarParameter() {defaultValue=492.4005, parameterName=Width2}
   Alpha=@n23
  )
  n25 = ScalarParameter() {defaultValue=243.8459, parameterName=WidthTip}
  n26 = TextureCoordinate()
  n27 = ScalarParameter() {defaultValue=0.2, parameterName=Constrotation}
  n28 = ComponentMask(None=WorldPosition() {worldPositionShaderOffset=WPT_ExcludeAllShaderOffsets}.XYZ) {b=true}
  n29 = Multiply(A=@n7, B=ScalarParameter() {defaultValue=-500, parameterName=WaveSpeed})
  n30 = Sine(None=Divide(A=Add(A=@n28, B=@n29), B=ScalarParameter() {defaultValue=2527.5166, parameterName=WaveLength}))
  n31 = ScalarParameter() {defaultValue=32.5929, parameterName=WaveAmplitude}
  n32 = Multiply(A=@n30, B=@n31)
  n33 = Multiply(A=@n12, B=ScalarParameter() {defaultValue=-445, parameterName=WaveSpeed2})
  n34 = Sine(None=Divide(A=Add(A=@n28, B=@n33), B=ScalarParameter() {defaultValue=1700.5785, parameterName=WaveLength2}))
  n35 = Multiply(A=@n34, B=LinearInterpolate(A=@n31, B=ScalarParameter() {defaultValue=1023.1006, parameterName=WaveAmplitude2}, Alpha=@n23))
  n36 = RotateAboutAxis(
   NormalizedRotationAxis=Constant3Vector() {constant=<rgba 0 0 1 0>}
   RotationAngle=Frac(None=Multiply(A=@n12, B=@n27))
   PivotPoint=Constant3Vector()
   Position=AppendVector(A=AppendVector(A=@n32, B=@n35), B=Constant())
  )
  n37 = LinearInterpolate(A=ScalarParameter() {defaultValue=-10, parameterName=MoveDown}, B=ScalarParameter() {defaultValue=-200, parameterName=MoveDown2}, Alpha=@n23)
  n38 = TextureCoordinate()
  n39 = Sign(None=Subtract(A=ComponentMask(None=@n38) {r=true}))
  n40 = Multiply(A=@n37, B=Multiply(A=@n30, B=@n39))
  n41 = Multiply(A=Multiply(A=@n34, B=@n39), B=@n37)
  n42 = Multiply(A=OneMinus(None=ComponentMask(None=@n26) {g=true}), B=AppendVector(A=ComponentMask(None=@n36) {g=true, r=true}, B=Add(A=@n40, B=@n41)))
  n43 = Add(A=WorldPosition() {worldPositionShaderOffset=WPT_ExcludeAllShaderOffsets}.XYZ, B=@n42)
  n44 = Multiply(
   A=Constant3Vector() {constant=<rgba 1 1 0 0>}
   B=Subtract(A=WorldPosition() {worldPositionShaderOffset=WPT_ExcludeAllShaderOffsets}.XYZ, B=ActorPositionWS())
  )
  n45 = Add(
   A=Add(
    A=MaterialFunctionCall(WidthBase=@n24, WidthTip=@n25, WorldPosition=@n43) {materialFunction=&SplineThicken}.WorldPositionOffset
    B=@n42
   )
   B=Multiply(A=Normalize(VectorInput=@n44), B=ScalarParameter() {defaultValue=1057.8693, parameterName=PushRadius})
  )
  n46 = Add(A=@n45, B=WorldPosition().XYZ)
  n47 = ActorPositionWS()
  n48 = Subtract(A=@n46, B=@n47)
  n49 = Multiply(A=Subtract(A=Distance(A=@n46, B=@n47), B=ScalarParameter() {defaultValue=1450, parameterName=RadiusOfSphere})) {constB=-1}
  n50 = Add(A=@n45, B=Multiply(A=Normalize(VectorInput=@n48), B=Clamp(None=@n49) {clampMode=CMODE_ClampMin}))
  n51 = VectorParameter() {parameterName=WPOOffsetLocal, defaultValue=<rgba 0 -1400 0 1>}
  Add(A=@n50, B=Transform(None=@n51.RGB) {transformSourceType=TRANSFORMSOURCE_Local})
  CameraVectorWS()
  n52 = Divide(
   A=Distance(A=WorldPosition() {worldPositionShaderOffset=WPT_ExcludeAllShaderOffsets}.XYZ, B=@n2)
   B=ScalarParameter() {defaultValue=1835.4244, parameterName=SphereFallOffDistance}
  )
  n53 = ScalarParameter() {defaultValue=0.4118, parameterName=SphereFallOff}
  n54 = Multiply(A=OneMinus(None=Power(Base=@n52, Exp=@n53)), B=VectorParameter() {parameterName=CoreGlow, defaultValue=<rgba 0.0409 0.1372 0.1979 1>}.RGB)
  n55 = ShadingPathSwitch(
   Default=VectorParameter() {parameterName=ColorFresnel, defaultValue=<rgba 0.103 0.3733 0.5 1>}.RGB
   Mobile=VectorParameter() {parameterName=ColorFresnel_Mobile, defaultValue=<rgba 0.103 0.3733 0.5 1>}.RGB
  )
  n56 = ComponentMask(None=TextureCoordinate()) {g=true}
  n57 = LinearInterpolate(
   A=ScalarParameter() {defaultValue=0.1333, parameterName=Low}
   B=ScalarParameter() {defaultValue=0.3984, parameterName=High}
   Alpha=DotProduct(A=Constant3Vector() {constant=<rgba 0 0 -1 0>}, B=@n20.Result)
  )
  n58 = VectorParameter() {parameterName=Color1, defaultValue=<rgba 0.0895 0.1298 0.2656 1>}
  n59 = Clamp(None=LinearInterpolate(
    A=ScalarParameter() {defaultValue=1.6616, parameterName=FalloffToptobottom}
    B=ScalarParameter() {parameterName=FalloffToptobottomLow}
    Alpha=@n56
   ))
  n60 = LinearInterpolate(
   A=VectorParameter() {parameterName=Color2x, defaultValue=<rgba 0.0304 0.1349 0.2656 1>}.RGB
   B=VectorParameter() {parameterName=Color2, defaultValue=<rgba 0.0347 0.0468 0.0729 1>}.RGB
   Alpha=@n59
  )
  n61 = ComponentMask(None=Max(
    A=TextureSample(UVs=@n16) {samplerType=SAMPLERTYPE_LinearColor, texture=/Figment_S08/Athena/Environments/Nexus/Textures/T_NexCoInk_D}.R
    B=TextureSample(UVs=@n18) {samplerType=SAMPLERTYPE_LinearColor, texture=&T_NexCoInk_D}.R
   )) {r=true}
  n62 = VectorParameter() {parameterName=Color1_Mobile, defaultValue=<rgba 0.0895 0.1298 0.2656 1>}
  n63 = LinearInterpolate(
   A=VectorParameter() {parameterName=Color2x_Mobile, defaultValue=<rgba 0.0304 0.1349 0.2656 1>}.RGB
   B=VectorParameter() {parameterName=Color2_Mobile, defaultValue=<rgba 0.0347 0.0468 0.0729 1>}.RGB
   Alpha=@n59
  )
  n64 = Add(
   A=Multiply(A=Multiply(A=@n55, B=@n56), B=Clamp(None=@n57))
   B=ShadingPathSwitch(Default=LinearInterpolate(A=@n58.RGB, B=@n60, Alpha=@n61), Mobile=LinearInterpolate(A=@n62.RGB, B=@n63, Alpha=@n61))
  )
  n65 = ScalarParameter() {defaultValue=3, parameterName=BrightnessMult}
  n66 = ScalarParameter() {defaultValue=1.75, parameterName=BrightnessMult2}
  n67 = ComponentMask(None=TextureCoordinate()) {g=true}
  Clamp(None=Multiply(A=Add(A=@n54, B=@n64), B=LinearInterpolate(A=@n65, B=@n66, Alpha=@n67))) {desc=Dont go pure black, maxDefault=1000, minDefault=0.05}
  Constant()
  MaterialFunctionCall(f3=ActorPositionWS()) {materialFunction=&AddComponents}
  MaterialFunctionCall!(Vector3=@n11, MaximumNumberOfDigits, UVs, DebugTextLocation RG_UpperRight BA_LowerLeft, Component Spacing) {materialFunction=/Engine/Functions/Engine_MaterialFunctions02/Utility/DebugFloat3Values}
  NamedRerouteDeclaration(None=@n50) {name=WPO, nodeColor=<rgba 0.8281 1 0 1>}
  Power(Exp=ScalarParameter() {defaultValue=3, parameterName=EmissiveContrast})
  n68 = ComponentMask(None=TextureCoordinate()) {g=true}
  n69 = Multiply(A=Multiply(A=@n68, B=OneMinus(None=@n68)), B=ScalarParameter() {defaultValue=2.8896, parameterName=FadeEnds})
  n70 = ScalarParameter() {defaultValue=0.1, parameterName=OpacityPOW}
  n71 = Multiply(
   A=MaterialFunctionCall(Transition Phase=Saturate(None=@n69), Height Texture=@n61, Contrast=ScalarParameter() {parameterName=Contrast}) {materialFunction=&HeightLerp}.Alpha
   B=Multiply(A=ScalarParameter() {defaultValue=1.3127, parameterName=Opacity}, B=Power(Base=@n61, Exp=@n70))
  )
  Reroute(None=DepthFade(Opacity=Clamp(None=@n71), FadeDistance=ScalarParameter() {defaultValue=138.8256, parameterName=Depthfade}))
  TransformPosition(None=@n51.RGB) {transformType=TRANSFORMPOSSOURCE_World}
  VectorParameter() {parameterName=CenterLocation, defaultValue=<rgba -3023.9985 -1764.4573 96 1>}
 ]
}
