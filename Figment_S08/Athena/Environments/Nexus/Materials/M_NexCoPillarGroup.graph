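# Default Lit material for Nexus pillar groups, building a bobbing wave offset and a distance-faded glow color.
# Time and Pivot Painter per-element random values shift object position into two summed sines, scaled into a vertical offset.
# Color blends by XY distance from the NexCollection CenterLocation and darkens with height above the actor origin.
# A second distance fade blends toward a background color masked by a view-angle term against world down.
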Material /Figment_S08/Athena/Environments/Nexus/Materials/M_NexCoPillarGroup
graph {
 properties {pixelDepthOffsetMode=PDOM_Legacy, usedShadingModels=Default Lit}
 nodes [
  Add()
  Constant2Vector()
  n1 = Multiply(
   A=MaterialFunctionCall() {materialFunction=/Engine/Functions/Engine_MaterialFunctions02/PivotPainter/PivotPainter_PerObjectData}.Random Value Per Element
   B=ScalarParameter() {defaultValue=23523.4395, parameterName=Random}
  )
  n2 = Time()
  n3 = Add(A=ObjectPositionWS(), B=Multiply(A=Add(A=@n1, B=@n2), B=ScalarParameter() {defaultValue=500, parameterName=WaveSpeed}))
  n4 = Divide(A=ComponentMask(None=@n3) {r=true}, B=ScalarParameter() {defaultValue=8000, parameterName=WaveLength})
  Frac(None=@n4)
  n5 = Divide(A=ComponentMask(None=@n3) {g=true}, B=ScalarParameter() {defaultValue=5000, parameterName=WaveLength2})
  Frac(None=@n5)
  n6 = VectorParameter() {parameterName=ColorEmissive2, defaultValue=<rgba 0.0496 0.1388 0.1615 1>}
  n7 = VectorParameter() {parameterName=ColorEmissive, defaultValue=<rgba 1.1068 2.2374 2.5 1>}
  n8 = Distance(
   A=ComponentMask(None=WorldPosition().XYZ) {g=true, r=true}
   B=ComponentMask(None=CollectionParameter() {collection=/Figment_S08/Environments/NexCollection, parameterName=CenterLocation}) {g=true, r=true}
  )
  n9 = ScalarParameter() {defaultValue=33326.8711, parameterName=DistanceFadeOffset}
  n10 = Saturate(None=Divide(A=Subtract(A=@n8, B=@n9), B=ScalarParameter() {defaultValue=25191.5605, parameterName=DistanceFade}))
  n11 = ComponentMask(None=Subtract(A=WorldPosition().XYZ, B=ActorPositionWS())) {b=true}
  n12 = Divide(A=@n11, B=ScalarParameter() {defaultValue=3977.4114, parameterName=HeightFade})
  n13 = CameraPositionWS()
  n14 = WorldPosition()
  n15 = DotProduct(A=Constant3Vector() {constant=<rgba 0 0 -1 0>}, B=Normalize(VectorInput=Subtract(A=@n13, B=@n14.XYZ)))
  n16 = Power(Base=OneMinus(None=Clamp(None=@n15)), Exp=ScalarParameter() {defaultValue=14, parameterName=Contrast_SoulsFade_Exp_3})
  n17 = Subtract(A=@n8, B=ScalarParameter() {defaultValue=735.8201, parameterName=DistanceFadeOffset2})
  n18 = ScalarParameter() {defaultValue=69649.5938, parameterName=DistanceFade2}
  LinearInterpolate(
   A=LinearInterpolate(A=LinearInterpolate(A=@n6.RGB, B=@n7.RGB, Alpha=@n10), B=Constant3Vector(), Alpha=Saturate(None=@n12))
   B=Multiply(A=VectorParameter() {parameterName=Color_Background_2, defaultValue=<rgba 2.7453 5 4.7508 1>}.RGB, B=Clamp(None=@n16))
   Alpha=Saturate(None=Divide(A=@n17, B=@n18))
  )
  n19 = Sine(None=@n4)
  n20 = Sine(None=@n5)
  Multiply(
   A=Multiply(A=Constant3Vector() {constant=<rgba 0 0 1 0>}, B=ScalarParameter() {defaultValue=1000, parameterName=Amplitude})
   B=Subtract(A=Add(A=@n19, B=@n20)) {constB=2}
  )
  Saturate(None=@n11)
  ScalarParameter() {parameterName=R}
  VectorParameter() {parameterName=Color, defaultValue=<rgba 0.0854 0.0558 0.125 1>}
  VectorParameter() {parameterName=ColorEmissiveFarFog, defaultValue=<rgba 2.7292 3.5422 4 1>}
  WorldPosition()
 ]
}
