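# Material function that paints a fake glowing Nexus core, centered on a world location, into incoming material attributes.
# Camera-facing virtual planes at two depth offsets project a cracked-glass texture, with one layer on Forward and Mobile shading paths.
# Cracks and Fresnel distort a radial distance that drives layered emissive falloffs and a mask blacking out underlying emissive and specular.
# A view-direction mask removes the effect when looking away from the center.
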MaterialFunction /Figment_S08/Athena/Environments/Nexus/Materials/MF_NexCoreGlowSimple_Figment
graph {nodes [
  n1 = FunctionInput(Preview=VectorParameter() {parameterName=SphereCenter, defaultValue=<rgba -3024 -1776 96 1>}.RGB) {bUsePreviewValueAsDefault=true, inputName=Centerlocation, sortPriority=2}
  n2 = Normalize(VectorInput=Subtract(A=@n1, B=CameraPositionWS()))
  n3 = VectorParameter() {parameterName=Axis, defaultValue=<rgba 0 0 1 1>}
  n4 = ScalarParameter() {defaultValue=-1000, parameterName=OffsetPlaneCore}
  n5 = Normalize(VectorInput=Subtract(A=@n1, B=CameraPositionWS()))
  n6 = MaterialFunctionCall!(
   Plane Normal Axis=@n2
   Plane X Axis=Normalize(VectorInput=CrossProduct(A=@n2, B=@n3.RGB))
   Plane Center=Add(A=@n1, B=Multiply(A=@n4, B=@n5))
   UV Scale=ScalarParameter() {defaultValue=1185.5463, parameterName=SizeProjectionTextureCore}
  ) {materialFunction=/Game/Athena/Environments/Nexus/Materials/VirtualPlaneCoordinatesAbs}
  n7 = Distance(A=@n6.UVs, B=Constant2Vector() {g=0.5, r=0.5})
  Add(A=@n7)
  CameraVectorWS()
  CameraVectorWS()
  Clamp(None=Floor(None=@n7))
  n8 = GetMaterialAttributes(None=FunctionInput() {inputType=FunctionInput_MaterialAttributes}) {attributeGetTypes [#1, #2]}
  n9 = Constant()
  n10 = TextureSample(UVs=@n6.UVs) {texture=/Game/Effects/Fort_Effects/Textures/Tile/T_Glass_Cracks_M}
  n11 = VectorParameter() {parameterName=Axis, defaultValue=<rgba 0 0 1 1>}
  n12 = ScalarParameter() {defaultValue=500, parameterName=OffsetCenter2Core}
  n13 = MaterialFunctionCall(
   Plane Normal Axis=@n2
   Plane X Axis=Normalize(VectorInput=CrossProduct(A=@n2, B=@n11.RGB))
   Plane Center=Add(A=@n1, B=Multiply(A=@n5, B=@n12))
   UV Scale=ScalarParameter() {defaultValue=1063.7999, parameterName=SizeProjectionTexture2Core}
  ) {materialFunction=&VirtualPlaneCoordinatesAbs}
  n14 = Add(A=@n10.R, B=TextureSample(UVs=Add(A=@n13.UVs) {constB=0.5}) {texture=&T_Glass_Cracks_M}.R)
  n15 = ShadingPathSwitch(Default=@n14, Deferred=@n14, Forward=@n10.R, Mobile=@n10.R)
  n16 = QualitySwitch(Default=@n15, High=@n15, Medium=@n15)
  n17 = ScalarParameter() {defaultValue=-0.1238, parameterName=Distortion}
  n18 = Fresnel(ExponentIn=ScalarParameter() {defaultValue=0.4967, parameterName=DistortionExp}) {baseReflectFraction=0}
  n19 = ScalarParameter() {defaultValue=0.2744, parameterName=DistortionNormal}
  n20 = Multiply(A=Add(A=Multiply(A=@n16, B=@n17), B=@n7), B=OneMinus(None=Multiply(A=@n18, B=@n19)))
  n21 = Power(
   Base=Divide(A=@n20, B=ScalarParameter() {defaultValue=0.7709, parameterName=DistanceFalloffBlackmask})
   Exp=ScalarParameter() {defaultValue=20.1444, parameterName=FresnelEXPBlackMask}
  )
  n22 = CameraVectorWS()
  n23 = Ceil(None=Saturate(None=DotProduct(A=@n2, B=@n22)))
  n24 = Clamp(None=Add(A=@n21, B=@n23))
  n25 = LinearInterpolate(A=@n9, B=@n8.EmissiveColor, Alpha=@n24)
  n26 = VectorParameter() {parameterName=EmColor1, defaultValue=<rgba 0.1636 0.4979 0.9948 1>}
  n27 = VectorParameter() {parameterName=EmColor2, defaultValue=<rgba 0 0 0 1>}
  n28 = ScalarParameter() {defaultValue=0.9241, parameterName=DistanceFalloff}
  n29 = Clamp(None=Power(Base=Divide(A=@n20, B=@n28), Exp=ScalarParameter() {defaultValue=0.3703, parameterName=FresnelEXP}))
  n30 = Multiply(
   A=VectorParameter() {parameterName=CoreGLow, defaultValue=<rgba 1.6276 2.7893 4 1>}.RGB
   B=ScalarParameter() {defaultValue=0.4667, parameterName=Glow}
  )
  n31 = Divide(A=@n20, B=ScalarParameter() {defaultValue=0.9691, parameterName=DistanceFalloff2})
  n32 = Clamp(None=Power(Base=OneMinus(None=@n31), Exp=ScalarParameter() {defaultValue=8, parameterName=FresnelEXP3}))
  n33 = ScalarParameter() {defaultValue=0.7881, parameterName=DistanceFalloff3}
  n34 = Power(Base=OneMinus(None=Divide(A=@n20, B=@n33)), Exp=ScalarParameter() {defaultValue=2.8268, parameterName=FresnelEXP4})
  n35 = Add(
   A=Add(A=LinearInterpolate(A=@n26.RGB, B=@n27.RGB, Alpha=@n29), B=Multiply(A=@n30, B=@n32))
   B=Multiply(A=VectorParameter() {parameterName=CoreGLow2, defaultValue=<rgba 0.8339 1.6354 2 1>}.RGB, B=Clamp(None=@n34))
  )
  n36 = Multiply(A=OneMinus(None=@n23), B=@n35)
  FunctionOutput(None=SetMaterialAttributes!(
    MaterialAttributes=@n8.MaterialAttributes
    EmissiveColor=Add(A=@n25, B=@n36) {constB=0}
    Specular=LinearInterpolate(A=@n9, B=@n8.Specular, Alpha=@n24)
   ) {attributeSetTypes [#1, #2]})
  n37 = Distance(A=@n1, B=WorldPosition().XYZ)
  n38 = Distance(A=ComponentMask(None=CameraPositionWS()) {g=true, r=true}, B=ComponentMask(None=@n1) {g=true, r=true})
  n39 = OneMinus(None=Clamp(None=Divide(A=@n37, B=@n38)))
  LinearInterpolate(A=Constant() {r=1}, Alpha=Ceil(None=Clamp(None=@n39)))
  MaterialFunctionCall() {materialFunction=&VirtualPlaneCoordinates}
  MaterialFunctionCall() {materialFunction=&VirtualPlaneCoordinates}
  Multiply(A=Constant() {r=1}, B=@n35)
  OneMinus(None=@n21)
  Reroute(None=@n39)
 ]}
