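# Material function for Nexus floor glow (or pillars via an isPillar switch) that replaces emissive with fake-depth glowing glass and data lines.
# VirtualPlaneCoordinates projects two offset layers of tinted crack textures, distorted by input normals, for parallax depth.
# Two crossed panning line textures are contrasted, tinted and faded in, but omitted on low quality, forward and mobile paths.
# The incoming emissive red channel masks intensity between floor and holes.
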MaterialFunction /Figment_S08/Athena/Environments/Nexus/Materials/MF_NexFloorGlow
graph {nodes [
  n1 = ScalarParameter() {defaultValue=-150, parameterName=OffsetPlane}
  n2 = CameraPositionWS()
  n3 = ComponentMask(None=@n2) {b=true}
  Add(A=@n1, B=@n3)
  n4 = ScalarParameter() {defaultValue=-250, parameterName=OffsetCenter2}
  Add(A=@n3, B=@n4)
  n5 = GetMaterialAttributes(None=FunctionInput() {inputType=FunctionInput_MaterialAttributes}) {attributeGetTypes [#1]}
  n6 = LinearInterpolate(
   A=ScalarParameter() {defaultValue=0.6286, parameterName=Floor}
   B=ScalarParameter() {defaultValue=1, parameterName=Holes}
   Alpha=ComponentMask(None=@n5.EmissiveColor) {r=true}
  )
  n7 = VectorParameter() {parameterName=Lines, defaultValue=<rgba 0.0625 2.3244 3 1>}
  n8 = CameraVectorWS()
  n9 = VectorParameter() {parameterName=Axis, defaultValue=<rgba 0 0 1 1>}
  n10 = StaticSwitchParameter(True=@n8, False=@n9.RGB) {parameterName=isPillar}
  n11 = StaticSwitchParameter(True=CrossProduct(A=@n8, B=@n9.RGB), False=VectorParameter() {parameterName=AxisX, defaultValue=<rgba 1 0 0 1>}.RGB) {parameterName=isPillar}
  n12 = AppendVector(A=Constant2Vector(), B=@n1)
  n13 = FunctionInput(Preview=VectorParameter() {parameterName=SphereCenter, defaultValue=<rgba -3024 -1776 96 1>}.RGB) {bUsePreviewValueAsDefault=true, inputName=Centerlocation, sortPriority=2}
  n14 = MaterialFunctionCall(
   Plane Normal Axis=@n10
   Plane X Axis=@n11
   Plane Center=StaticSwitchParameter(True=Add(A=@n12, B=@n13), False=Add(A=@n2, B=@n12)) {parameterName=isPillar}
   UV Scale=ScalarParameter() {defaultValue=50, parameterName=SizeProjectionTexture}
  ) {materialFunction=&VirtualPlaneCoordinates}
  n15 = Multiply(A=@n14.UVs, B=ScalarParameter() {defaultValue=0.1, parameterName=TilingLines})
  n16 = AppendVector(A=ScalarParameter() {defaultValue=0.01, parameterName=SpeedPanLines2}, B=ScalarParameter() {defaultValue=1, parameterName=SpeedPanLines})
  n17 = Panner(Coordinate=@n15, Speed=@n16)
  n18 = Panner(Coordinate=MaterialFunctionCall(XY=@n15) {materialFunction=&Swizzle}.YX, Speed=@n16)
  n19 = Power(
   Base=Add(
    A=TextureSample(UVs=@n17) {samplerType=SAMPLERTYPE_LinearColor, texture=/Figment_S08/Athena/Environments/Nexus/Textures/T_MovingLines}.R
    B=TextureSample(UVs=@n18) {samplerType=SAMPLERTYPE_LinearColor, texture=&T_MovingLines}.R
   )
   Exp=ScalarParameter() {defaultValue=5, parameterName=LinesContrast}
  )
  n20 = Multiply(
   A=ScalarParameter() {defaultValue=0.2, parameterName=DistortNormals}
   B=ComponentMask(None=FunctionInput() {bUsePreviewValueAsDefault=true, inputName=Normal, sortPriority=12}) {g=true, r=true}
  )
  n21 = Add(A=@n14.UVs, B=@n20)
  n22 = AppendVector(A=Constant2Vector(), B=@n4)
  n23 = Constant3Vector() {constant=<rgba 0.99 0.99 1 0>}
  n24 = StaticSwitchParameter(True=Add(A=@n22, B=@n13), False=Add(A=Multiply(A=@n2, B=@n23), B=@n22)) {parameterName=isPillar}
  n25 = ScalarParameter() {defaultValue=200, parameterName=SizeProjectionTexture2}
  n26 = Add(
   A=MaterialFunctionCall(XY=MaterialFunctionCall(Plane Normal Axis=@n10, Plane X Axis=@n11, Plane Center=@n24, UV Scale=@n25) {materialFunction=&VirtualPlaneCoordinates}.UVs) {materialFunction=&Swizzle}.YX
   B=@n20
  )
  n27 = Add(
   A=Multiply(
    A=VectorParameter() {parameterName=GlassEmissive1, defaultValue=<rgba 0.3155 0.4952 0.9792 1>}.RGB
    B=TextureSample(UVs=@n21) {texture=/Game/Effects/Fort_Effects/Textures/Tile/T_Glass_Cracks_M}.R
   )
   B=Multiply(
    A=VectorParameter() {parameterName=GlassEmissive2, defaultValue=<rgba 0.4547 0.7539 0.9115 1>}.RGB
    B=TextureSample(UVs=@n26) {texture=&T_Glass_Cracks_M}.R
   )
  )
  n28 = Add(A=Multiply(A=Multiply(A=@n7.RGB, B=@n19), B=ScalarParameter() {parameterName=FadeInDataLines}), B=@n27)
  n29 = ShadingPathSwitch(Default=@n28, Deferred=@n28, Forward=@n27, Mobile=@n27)
  n30 = QualitySwitch(Default=@n29, Low=@n27, High=@n29, Medium=@n29)
  FunctionOutput(None=SetMaterialAttributes!(MaterialAttributes=@n5.MaterialAttributes, EmissiveColor=Multiply(A=@n6, B=@n30)) {attributeSetTypes [#1]})
  Multiply(A=@n2, B=Constant3Vector() {constant=<rgba 0.995 0.995 1 0>})
 ]}
