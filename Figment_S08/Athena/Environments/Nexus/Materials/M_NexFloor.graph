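# Master material for the Figment S08 Nexus floor, with switches for pillar, stairs and animated wave variants.
# Height-blends two tiled bump-offset layers using panning noise, feeding color, normals and emissive veins into floor glow.
# Adds chromatic Fresnel, optional core glow and white fade, then a radial color fade around the collection's center, scaled by camera height.
# Forward/mobile paths simplify texturing; lower quality levels remap roughness and specular by camera distance.
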Material /Figment_S08/Athena/Environments/Nexus/Materials/M_NexFloor
graph {
 properties {bUseMaterialAttributes=true, bUsedWithNanite=true, pixelDepthOffsetMode=PDOM_Legacy, usedShadingModels=Default Lit}
 nodes [
  Multiply()
  n1 = VectorParameter() {parameterName=ColorCoreFade, defaultValue=<rgba 0 0 0 1>}
  n2 = TextureCoordinate()
  n3 = Divide(A=ComponentMask(None=ObjectPositionWS()) {g=true, r=true}, B=ScalarParameter() {defaultValue=0.3523, parameterName=RandowmOffsetUV})
  n4 = ScalarParameter() {defaultValue=24.9361, parameterName=Tiling}
  n5 = Multiply(A=StaticSwitchParameter(True=Add(A=@n2, B=@n3), False=@n2) {parameterName=IsPillar}, B=@n4)
  n6 = BumpOffset(
   Coordinate=@n5
   Height=TextureSample(UVs=@n5) {samplerType=SAMPLERTYPE_Masks, texture=/Figment_S08/Athena/Environments/Nexus/Textures/NexFloor_2_M}.G
   HeightRatioInput=ScalarParameter() {defaultValue=0.01, parameterName=BumpOffset}
  )
  n7 = ShadingPathSwitch(Default=@n6, Deferred=@n6, Forward=@n5, Mobile=@n5)
  n8 = TextureSample(UVs=@n7) {samplerType=SAMPLERTYPE_Masks, texture=&NexFloor_2_M}
  n9 = Multiply(A=TextureCoordinate(), B=Multiply(A=@n4, B=ScalarParameter() {defaultValue=9.9396, parameterName=Tiling2}))
  n10 = BumpOffset(
   Coordinate=@n9
   Height=TextureSample(UVs=@n9) {samplerType=SAMPLERTYPE_Masks, texture=/Game/Athena/Environments/Nexus/Textures/NexFloor_M}.G
   HeightRatioInput=ScalarParameter() {defaultValue=0.01, parameterName=BumpOffset2}
  )
  n11 = ShadingPathSwitch(Default=@n10, Deferred=@n10, Forward=@n9, Mobile=@n9)
  n12 = TextureSample(UVs=@n11) {samplerType=SAMPLERTYPE_Masks, texture=&NexFloor_M}
  n13 = ScalarParameter() {defaultValue=1.6489, parameterName=Tiling3}
  n14 = Panner(
   Coordinate=Multiply(A=TextureCoordinate(), B=Multiply(A=@n4, B=@n13))
   Speed=ScalarParameter() {defaultValue=0.0571, parameterName=SpeedPanning1}
  )
  n15 = Add(
   A=OneMinus(None=@n8.R)
   B=LinearInterpolate(
    A=ScalarParameter() {defaultValue=1, parameterName=Transition}
    B=ScalarParameter() {parameterName=Transition2}
    Alpha=TextureSample(UVs=@n14) {texture=/Figment_S08/Effects/Fort_Effects/Textures/Misc/T_Tiling_Noise_01}.R
   )
  )
  n16 = MaterialFunctionCall!(Diffuse 1, Diffuse 2, Transistion Phase=@n15, Height Texture 1=OneMinus(None=@n8.G), Height Texture 2=@n12.G, Contrast) {materialFunction=/Engine/Functions/Engine_MaterialFunctions02/Texturing/HeightLerpWithTwoHeightMaps}
  n17 = MaterialFunctionCall(Transition Phase=OneMinus(None=@n15), Height Texture=@n12.G) {materialFunction=&HeightLerp}
  n18 = ShadingPathSwitch(Default=@n16.Alpha, Deferred=@n16.Alpha, Forward=@n17.Alpha, Mobile=@n17.Alpha)
  n19 = QualitySwitch(Default=@n18, Low=@n17.Alpha, High=@n18, Medium=@n18)
  n20 = Multiply(
   A=@n8.R
   B=MaterialFunctionCall(Float3=LinearInterpolate(A=@n8.RGB, B=@n12.RGB, Alpha=@n19)) {materialFunction=&BreakOutFloat3Components}.R
  )
  n21 = ScalarParameter() {defaultValue=0.2, parameterName=AO}
  n22 = Multiply(
   A=LinearInterpolate(
    A=VectorParameter() {parameterName=Color2, defaultValue=<rgba 0.3412 0.0583 0.5677 1>}.RGB
    B=VectorParameter() {parameterName=Color, defaultValue=<rgba 0.9271 0.9271 0.9271 1>}.RGB
    Alpha=@n20
   )
   B=Saturate(None=Add(A=@n21, B=@n20))
  )
  n23 = ComponentMask(None=CollectionParameter() {collection=/Figment_S08/Environments/NexCollection, parameterName=CenterLocation}) {b=true, g=true, r=true}
  n24 = Distance(A=@n23, B=WorldPosition().XYZ)
  n25 = ScalarParameter() {defaultValue=1500, parameterName=OffsetCenterFade}
  n26 = Clamp(None=Divide(A=Subtract(A=@n24, B=@n25), B=ScalarParameter() {defaultValue=500, parameterName=OffsetCenterFadeDistance}))
  n27 = Multiply(A=TextureCoordinate(), B=Multiply(A=@n4, B=ScalarParameter() {defaultValue=7.3023, parameterName=Tiling4}))
  n28 = ScalarParameter() {defaultValue=0.1, parameterName=SpeedPanning2}
  n29 = Multiply(
   A=Power(Base=OneMinus(None=@n20), Exp=ScalarParameter() {defaultValue=25.737, parameterName=EmissivePower})
   B=TextureSample(UVs=Panner(Coordinate=@n27, Speed=@n28)) {texture=&T_Tiling_Noise_01}.R
  )
  n30 = Transform(None=Constant3Vector() {constant=<rgba 0 0 1 0>}) {transformSourceType=TRANSFORMSOURCE_World, transformType=TRANSFORM_Tangent}
  n31 = TextureSample(UVs=@n7) {samplerType=SAMPLERTYPE_Normal, texture=/Figment_S08/Athena/Environments/Nexus/Textures/NexFloor_2_N}
  n32 = ScalarParameter() {defaultValue=7.0369, parameterName=TilingMicro}
  n33 = MaterialFunctionCall(
   BaseNormal=@n31.RGB
   AdditionalNormal=TextureSample(UVs=Multiply(A=@n5, B=@n32)) {samplerType=SAMPLERTYPE_Normal, texture=&NexFloor_2_N}.RGB
  ) {materialFunction=&BlendAngleCorrectedNormals}
  n34 = ShadingPathSwitch(Default=@n33.Result, Deferred=@n33.Result, Forward=@n31.RGB, Mobile=@n31.RGB)
  n35 = Constant3Vector() {constant=<rgba 0 0 1 0>}
  n36 = TextureSample(UVs=@n11) {samplerType=SAMPLERTYPE_Normal, texture=/Game/Athena/Environments/Nexus/Textures/NexFloor_N}
  n37 = MaterialFunctionCall(
   BaseNormal=QualitySwitch(Default=@n34, Low=@n31.RGB, High=@n34, Medium=@n34)
   AdditionalNormal=MaterialFunctionCall(Normal=LinearInterpolate(A=@n35, B=@n36.RGB, Alpha=@n19), Flatness=ScalarParameter() {defaultValue=-0.2276, parameterName=NormalFlatten}) {materialFunction=&FlattenNormal}.Result
  ) {materialFunction=&BlendAngleCorrectedNormals}
  n38 = MakeMaterialAttributes!(
   BaseColor=StaticSwitchParameter(True=LinearInterpolate(A=@n1.RGB, B=@n22, Alpha=@n26), False=@n22) {parameterName=isStairs}
   Metallic
   Specular=LinearInterpolate(A=ScalarParameter() {defaultValue=-0.1, parameterName=Specular}, B=ScalarParameter() {defaultValue=1, parameterName=Specular2}, Alpha=@n20)
   Roughness=LinearInterpolate(
    A=ScalarParameter() {defaultValue=0.3762, parameterName=Roughness}
    B=ScalarParameter() {defaultValue=-0.0571, parameterName=Roughness2}
    Alpha=@n20
   )
   Anisotropy
   EmissiveColor=LinearInterpolate(A=Constant3Vector(), B=ScalarParameter() {defaultValue=500, parameterName=EmissiveMask}, Alpha=Clamp(None=@n29))
   Opacity
   OpacityMask
   Normal=StaticSwitchParameter(
    True=MaterialFunctionCall(BaseNormal=@n30, AdditionalNormal=@n37.Result) {materialFunction=&BlendAngleCorrectedNormals}.Result
    False=@n37.Result
   ) {parameterName=NormalForcedUp}
   Tangent
   WorldPositionOffset
   SubsurfaceColor
   ClearCoat
   ClearCoatRoughness
   AmbientOcclusion
   Refraction
   CustomizedUVs_0
   CustomizedUVs_1
   CustomizedUVs_2
   CustomizedUVs_3
   CustomizedUVs_4
   CustomizedUVs_5
   CustomizedUVs_6
   CustomizedUVs_7
   PixelDepthOffset
   ShadingModel
   Displacement
  )
  n39 = GetMaterialAttributes(None=@n38) {attributeGetTypes [#1]}
  n40 = MaterialFunctionCall!(In=@n39.MaterialAttributes, Centerlocation=ObjectPositionWS(), Normal=@n39.MaterialAttributes) {materialFunction=&/MF_NexFloorGlow}
  n41 = ShadingPathSwitch(Default=@n40.Result, Deferred=@n40.Result, Forward=@n38, Mobile=@n38)
  QualitySwitch(Default=@n41, Low=@n38, High=@n41, Medium=@n41)
  n42 = MakeMaterialAttributes(
   BaseColor=VectorParameter() {parameterName=ColorTop, defaultValue=<rgba 0.0656 0.0115 0.1458 1>}.RGB
   Specular=Constant() {r=0.5}
   Roughness=ScalarParameter() {parameterName=RoughnessTop}
   EmissiveColor=Constant3Vector()
   Normal=Constant3Vector() {constant=<rgba 0 0 1 0>}
  )
  n43 = Constant3Vector() {constant=<rgba 0 0 1 0>}
  n44 = MaterialFunctionCall() {materialFunction=/Engine/Functions/Engine_MaterialFunctions02/WorldPositionOffset/ObjectScale}
  n45 = Multiply(A=ScalarParameter() {defaultValue=-581.5906, parameterName=HeightforGradient}, B=@n44.Scale Z)
  n46 = ScalarParameter() {defaultValue=354.2845, parameterName=GradientSize}
  n47 = AppendVector(A=@n46, B=@n46)
  n48 = ScalarParameter() {defaultValue=1709.8967, parameterName=GradientSizeZ}
  n49 = Divide(A=Subtract(A=WorldPosition().XYZ, B=Multiply(A=@n43, B=@n45)), B=Multiply(A=AppendVector(A=@n47, B=@n48), B=@n44.Scale Z))
  n50 = LinearInterpolate(
   A=TextureSample(UVs=ComponentMask(None=@n49) {b=true, r=true}) {samplerType=SAMPLERTYPE_Masks, texture=/Game/Athena/Environments/Nexus/Textures/StreakGradient_M}.RGB
   B=TextureSample(UVs=ComponentMask(None=@n49) {b=true, g=true}) {samplerType=SAMPLERTYPE_Masks, texture=&StreakGradient_M}.RGB
   Alpha=ComponentMask(None=VertexNormalWS()) {r=true}
  )
  n51 = Saturate(None=LinearInterpolate(
    A=ScalarParameter() {defaultValue=-0.4335, parameterName=C1}
    B=ScalarParameter() {defaultValue=1, parameterName=C2}
    Alpha=MaterialFunctionCall(Float3=@n50) {materialFunction=&BreakOutFloat3Components}.R
   ))
  n52 = GetMaterialAttributes(None=StaticSwitchParameter(
    True=MaterialFunctionCall!(Base Material=@n42, Top Material=@n40.Result, Alpha=@n51) {materialFunction=/Engine/Functions/MaterialLayerFunctions/MatLayerBlend_Standard}.Blended Material
    False=@n40.Result
   ) {parameterName=IsPillar}) {attributeGetTypes [#2]}
  n53 = Multiply(A=Constant3Vector() {constant=<rgba 0 0 1 0>}, B=ScalarParameter() {defaultValue=200, parameterName=Amplitude})
  n54 = Add(A=ObjectPositionWS(), B=Multiply(A=Time(), B=ScalarParameter() {defaultValue=500, parameterName=WaveSpeed}))
  n55 = Divide(A=ComponentMask(None=@n54) {r=true}, B=ScalarParameter() {defaultValue=8000, parameterName=WaveLength})
  n56 = Divide(A=ComponentMask(None=@n54) {g=true}, B=ScalarParameter() {defaultValue=5000, parameterName=WaveLength2})
  n57 = Subtract(A=Add(A=Sine(None=@n55), B=Sine(None=@n56))) {constB=2}
  n58 = Fresnel(ExponentIn=ScalarParameter() {defaultValue=3, parameterName=CromaticExponent}) {baseReflectFraction=0}
  n59 = Multiply(A=TextureSample(UVs=OneMinus(None=@n58)) {texture=/Game/Effects/Fort_Effects/Textures/Tile/T_Inky_Smoke_Tile_Chromatic}.RGB, B=@n58)
  n60 = ShadingPathSwitch(Default=@n59, Mobile=Constant())
  n61 = ScalarParameter() {parameterName=HideChromaticRingOnLow}
  n62 = SetMaterialAttributes!(
   MaterialAttributes=@n52.MaterialAttributes
   World Position Offset=StaticSwitchParameter(True=Multiply(A=@n53, B=@n57), False=Constant3Vector()) {parameterName=IsAnimating}
   Emissive Color=Add(A=LinearInterpolate(A=@n59, B=@n60, Alpha=@n61), B=@n52.Emissive Color)
  ) {attributeSetTypes [#3, #2]}
  n63 = StaticSwitchParameter(
   True=MaterialFunctionCall!(In=@n62, Centerlocation) {materialFunction=&/MF_NexCoreGlowSimple_Figment}.Result
   False=@n62
  ) {parameterName=UseCoreGlow}
  n64 = MakeMaterialAttributes(
   BaseColor=VectorParameter() {parameterName=WhiteFade, defaultValue=<rgba 0.9948 0.9948 0.9948 1>}.RGB
   Roughness=ScalarParameter() {parameterName=R}
   EmissiveColor=VectorParameter() {parameterName=WhiteFadeEm, defaultValue=<rgba 0.9948 0.9948 0.9948 1>}.RGB
  )
  n65 = Distance(A=ComponentMask(None=WorldPosition().XYZ) {g=true, r=true}, B=Constant2Vector())
  n66 = ScalarParameter() {defaultValue=5000, parameterName=OffsetWhiteFade}
  n67 = Saturate(None=Divide(A=Subtract(A=@n65, B=@n66), B=ScalarParameter() {defaultValue=5000, parameterName=OffsetWhiteFadeDistance}))
  n68 = GetMaterialAttributes(None=StaticSwitchParameter(
    True=MaterialFunctionCall(Base Material=@n63, Top Material=@n64, Alpha=@n67) {materialFunction=&MatLayerBlend_Standard}.Blended Material
    False=@n63
   ) {parameterName=UseFlowFadeToWhite}) {attributeGetTypes [
    #4
    #2
    #5
    #6
   ]}
  n69 = VectorParameter() {parameterName=FadeRadialColor, defaultValue=<rgba 0.7867 1.1575 1.5 1>}
  n70 = WorldPosition()
  n71 = Subtract(
   A=Distance(A=ComponentMask(None=@n70.XYZ) {g=true, r=true}, B=ComponentMask(None=@n23) {g=true, r=true})
   B=ScalarParameter() {defaultValue=1, parameterName=FadeRadialOffset}
  )
  n72 = ComponentMask(None=CameraPositionWS()) {b=true}
  n73 = ComponentMask(None=@n23) {b=true}
  n74 = Divide(
   A=Subtract(A=Subtract(A=@n72, B=@n73), B=ScalarParameter() {parameterName=CamerHeightOffset})
   B=ScalarParameter() {defaultValue=16000, parameterName=CamerHeightDistance}
  )
  n75 = LinearInterpolate(
   A=ScalarParameter() {defaultValue=31779.7773, parameterName=FadeRadialDistance}
   B=ScalarParameter() {defaultValue=10779.7773, parameterName=FadeRadialDistance2}
   Alpha=VertexInterpolator(VS=Saturate(None=@n74)).PS
  )
  n76 = Clamp(None=Power(Base=Divide(A=@n71, B=@n75)) {constExponent=3})
  n77 = ScalarParameter() {defaultValue=1, desc=Effects Low/Medium, parameterName=RoughnessDistance.FarMulti}
  n78 = ScalarParameter() {defaultValue=1, desc=Effects Low/Medium, parameterName=RoughnessDistance.NearMulti}
  n79 = SphereMask(
   A=ComponentMask(None=CameraPositionWS()) {g=true, r=true}
   B=WorldPosition().XY
   Radius=ScalarParameter() {desc=Effects Low/Medium, parameterName=RoughnessDistanceMask.Radius}
   Hardness=ScalarParameter() {defaultValue=1, desc=Effects Low/Medium, parameterName=RoughnessDistanceMask.Hardness}
  )
  n80 = Add(
   A=Multiply(A=@n68.Roughness, B=LinearInterpolate(A=@n77, B=@n78, Alpha=@n79))
   B=LinearInterpolate(
    A=ScalarParameter() {desc=Effects Low/Medium, parameterName=RoughnessDistance.FarAdd}
    B=ScalarParameter() {desc=Effects Low/Medium, parameterName=RoughnessDistance.NearAdd}
    Alpha=@n79
   )
  )
  n81 = Multiply(
   A=@n68.Specular
   B=LinearInterpolate(
    A=ScalarParameter() {defaultValue=1, desc=Effects Low/Medium, parameterName=SpecDistance.FarMulti}
    B=ScalarParameter() {defaultValue=1, desc=Effects Low/Medium, parameterName=SpecDistance.NearMulti}
    Alpha=@n79
   )
  )
  SetMaterialAttributes!(
   MaterialAttributes=@n68.MaterialAttributes
   Base Color=LinearInterpolate(A=@n68.Base Color, B=@n69.RGB, Alpha=@n76)
   Emissive Color=LinearInterpolate(A=@n68.Emissive Color, B=@n69.RGB, Alpha=@n76)
   Roughness=QualitySwitch(Default=@n80, Low=@n80, High=@n68.Roughness, Medium=@n80, Epic=@n68.Roughness)
   Specular=QualitySwitch(Default=@n81, Low=@n81, High=@n68.Specular, Medium=@n81, Epic=@n68.Specular)
  ) {attributeSetTypes [
    #4
    #2
    #5
    #6
   ]}
 ]
}
