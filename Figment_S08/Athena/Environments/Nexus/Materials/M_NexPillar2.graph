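# Two-sided, lit material for a Nexus pillar whose Animation parameter morphs it from a flattened block into separated, moving pieces.
# At zero, vertices beyond DistanceY squash toward side planes and normals snap to box axes.
# At one, Pivot Painter data offsets and rotates pieces with time-based wobble and corrected normals.
# Fresnel-mapped chromatic smoke adds masked emissive, and MF_NexPillarGlow post-processes the attributes using center and Animation.
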Material /Figment_S08/Athena/Environments/Nexus/Materials/M_NexPillar2
graph {
 properties {
  allowTranslucentCustomDepthWrites=true
  bTangentSpaceNormal=false
  bUseMaterialAttributes=true
  pixelDepthOffsetMode=PDOM_Legacy
  translucencyLightingMode=TLM_Surface
  twoSided=true
  usedShadingModels=Default Lit
 }
 nodes [
  n1 = VectorParameter() {parameterName=BaseEmissive, defaultValue=<rgba 0.0556 0.0775 0.1354 1>}
  n2 = TextureSample() {samplerType=SAMPLERTYPE_Masks, texture=/Figment_S08/Athena/Environments/Nexus/Textures/T_NexPillar2_D}
  n3 = Fresnel(ExponentIn=ScalarParameter() {defaultValue=0.1315, parameterName=CromaticExponent}) {baseReflectFraction=0}
  n4 = Multiply(A=@n2.R, B=TextureSample(UVs=OneMinus(None=@n3)) {texture=/Game/Effects/Fort_Effects/Textures/Tile/T_Inky_Smoke_Tile_Chromatic}.RGB)
  n5 = Transform(None=Constant3Vector() {constant=<rgba 0 1 0 0>}) {transformSourceType=TRANSFORMSOURCE_Local}
  n6 = Transform(None=Subtract(A=WorldPosition().XYZ, B=ObjectPositionWS())) {transformSourceType=TRANSFORMSOURCE_World, transformType=TRANSFORM_Local}
  n7 = ComponentMask(None=@n6) {g=true}
  n8 = ScalarParameter() {defaultValue=15, parameterName=DistanceY}
  n9 = Subtract(A=@n7, B=@n8)
  n10 = Clamp(None=Ceil(None=@n9))
  n11 = LinearInterpolate(A=Constant3Vector() {constant=<rgba 0 0 1 0>}, B=@n5, Alpha=@n10)
  n12 = Multiply(A=@n5) {constB=-1}
  n13 = Subtract(A=Multiply(A=@n7) {constB=-1}, B=@n8)
  n14 = Clamp(None=Ceil(None=@n13))
  n15 = Transform(None=Constant3Vector() {constant=<rgba 1 0 0 0>}) {transformSourceType=TRANSFORMSOURCE_Local}
  n16 = ComponentMask(None=@n6) {r=true}
  n17 = ScalarParameter() {defaultValue=89, parameterName=DistanceX}
  n18 = Ceil(None=Subtract(A=@n16, B=@n17))
  n19 = Subtract(A=Multiply(A=@n16) {constB=-1}, B=@n17)
  n20 = LinearInterpolate(
   A=LinearInterpolate(A=LinearInterpolate(A=@n11, B=@n12, Alpha=@n14), B=@n15, Alpha=Clamp(None=@n18))
   B=Multiply(A=@n15) {constB=-1}
   Alpha=Clamp(None=Ceil(None=@n19))
  )
  n21 = TextureSample() {samplerType=SAMPLERTYPE_Normal, texture=/Figment_S08/Athena/Environments/Nexus/Textures/T_NexPillar2_N}
  n22 = MaterialFunctionCall() {materialFunction=/Engine/Functions/Engine_MaterialFunctions02/PivotPainter/PivotPainter_PerObjectData}
  n23 = Add(A=@n22.Random Value Per Element, B=Time())
  n24 = Multiply(A=@n23, B=ScalarParameter() {defaultValue=0.5, parameterName=ConstantRotationSpeed2})
  n25 = Normalize(VectorInput=LinearInterpolate(
    A=VectorParameter() {parameterName=RotAxis1, defaultValue=<rgba 1 1 0 1>}.RGB
    B=VectorParameter() {parameterName=RotAxis2, defaultValue=<rgba 0 1 1 1>}.RGB
    Alpha=Sine(None=@n24)
   ))
  n26 = Multiply(A=@n23, B=ScalarParameter() {defaultValue=0.4476, parameterName=ConstantRotationSpeed})
  n27 = Multiply(A=OneMinus(None=@n2.B), B=Multiply(A=Sine(None=@n26), B=ScalarParameter() {defaultValue=0.01, parameterName=ContantRotationAngle}))
  n28 = VertexNormalWS()
  n29 = MaterialFunctionCall(
   BaseNormal=VertexInterpolator(VS=MaterialFunctionCall(Rotation Axis=@n25, Rotation Angle=@n27, World Space Vertex Normals=@n28) {materialFunction=&FixRotateAboutAxisNormals}.New Tangent Space Vertex Normal).PS
   AdditionalNormal=@n21.RGB
  ) {materialFunction=&BlendAngleCorrectedNormals}
  n30 = ShadingPathSwitch(Default=@n29.Result, Deferred=@n29.Result, Forward=@n21.RGB, Mobile=@n21.RGB)
  n31 = Transform(None=QualitySwitch(Default=@n30, Low=@n21.RGB, High=@n30, Medium=@n30))
  n32 = Clamp(None=ScalarParameter() {parameterName=Animation})
  n33 = Clamp(None=Multiply(A=ScalarParameter() {defaultValue=10, parameterName=NormalSwitchBias}, B=@n32))
  n34 = Clamp(None=@n9) {clampMode=CMODE_ClampMin}
  n35 = ScalarParameter() {defaultValue=-2, parameterName=BiasSurfaceposition}
  n36 = Multiply(A=@n35) {constB=-1}
  n37 = AppendVector(A=Constant(), B=Add(A=Multiply(A=@n34) {constB=-0.99}, B=Multiply(A=@n10, B=@n36)))
  n38 = Constant()
  n39 = Clamp(None=@n13) {clampMode=CMODE_ClampMin}
  n40 = AppendVector(A=Constant(), B=Add(A=Multiply(A=@n35, B=@n14), B=Multiply(A=@n39) {constB=0.99}))
  n41 = Constant()
  n42 = Transform(None=Add(A=AppendVector(A=@n37, B=@n38), B=AppendVector(A=@n40, B=@n41))) {transformSourceType=TRANSFORMSOURCE_Local}
  n43 = Clamp(None=@n32)
  n44 = TextureSample(UVs=TextureCoordinate() {coordinateIndex=1}) {samplerType=SAMPLERTYPE_LinearColor, texture=/Figment_S08/Athena/Environments/Nexus/Textures/Box227_rgb_PivotPos_a_ParentIndexInt_UV_2}
  n45 = MaterialFunctionCall!(Position RGB=@n44.RGB) {materialFunction=/Engine/Functions/Engine_MaterialFunctions02/PivotPainter2/ms_PivotPainter2_DecodePostion}
  n46 = Transform(None=Constant3Vector() {constant=<rgba 0 1 0 0>}) {transformSourceType=TRANSFORMSOURCE_Local}
  n47 = Add(
   A=ActorPositionWS()
   B=Multiply(
    A=VectorParameter() {parameterName=Center, defaultValue=<rgba 0 0 145.979 1>}.RGB
    B=MaterialFunctionCall() {materialFunction=/Engine/Functions/Engine_MaterialFunctions02/WorldPositionOffset/ObjectScale}.Scale Z
   )
  )
  n48 = Transform(None=Normalize(VectorInput=Subtract(A=@n22.Pivot Position, B=@n47))) {transformSourceType=TRANSFORMSOURCE_World, transformType=TRANSFORM_Local}
  n49 = Multiply(A=ComponentMask(None=@n48) {r=true}, B=ScalarParameter() {defaultValue=0.0571, parameterName=RotationX})
  n50 = WorldPosition()
  n51 = Constant3Vector() {constant=<rgba 0 0 1 0>}
  n52 = Multiply(
   A=Multiply(A=ComponentMask(None=@n48) {g=true}, B=ComponentMask(None=@n48) {r=true})
   B=ScalarParameter() {defaultValue=0.4667, parameterName=RotationY}
  )
  n53 = WorldPosition()
  n54 = Add(
   A=Multiply(A=Subtract(A=@n22.Pivot Position, B=@n45.Result), B=ScalarParameter() {defaultValue=1.0585, parameterName=Offset})
   B=Add(
    A=RotateAboutAxis(NormalizedRotationAxis=@n46, RotationAngle=@n49, PivotPoint=@n45.Result, Position=@n50.XYZ)
    B=RotateAboutAxis(NormalizedRotationAxis=@n51, RotationAngle=@n52, PivotPoint=@n45.Result, Position=@n53.XYZ)
   )
  )
  n55 = Add(A=@n54, B=WorldPosition().XYZ)
  n56 = Normalize(VectorInput=Subtract(A=WorldPosition().XYZ, B=ObjectPositionWS()))
  n57 = ScalarParameter() {defaultValue=2, parameterName=Scale}
  n58 = Add(
   A=Add(A=@n54, B=RotateAboutAxis(NormalizedRotationAxis=@n25, RotationAngle=@n27, PivotPoint=@n45.Result, Position=@n55))
   B=Multiply(A=Multiply(A=@n56, B=@n57), B=@n2.G)
  )
  MaterialFunctionCall!(
   In=MakeMaterialAttributes!(
    BaseColor=VectorParameter() {parameterName=Color, defaultValue=<rgba 0.001 0.0051 0.05 1>}.RGB
    Metallic
    Specular=ScalarParameter() {defaultValue=1, parameterName=Spec}
    Roughness=Constant()
    Anisotropy
    EmissiveColor=Add(A=@n1.RGB, B=@n4)
    Opacity
    OpacityMask
    Normal=LinearInterpolate(A=@n20, B=@n31, Alpha=@n33)
    Tangent
    WorldPositionOffset=LinearInterpolate(A=@n42, B=@n58, Alpha=@n43)
    SubsurfaceColor
    ClearCoat
    ClearCoatRoughness
    AmbientOcclusion
    Refraction
    CustomizedUVs_0
    CustomizedUVs_1
    CustomizedUVs_2
    CustomizedUVs_3
    CustomizedUVs_4
    CustomizedUVs_5
    CustomizedUVs_6
    CustomizedUVs_7
    PixelDepthOffset
    ShadingModel
    Displacement
   )
   Centerlocation=@n47
   Animation=@n43
   isWeapon
  ) {materialFunction=&/MF_NexPillarGlow}
  n59 = Multiply(A=Time(), B=ScalarParameter() {defaultValue=0.1, parameterName=TimeSpeed})
  n60 = Clamp(None=Multiply(A=Frac(None=@n59), B=ScalarParameter() {defaultValue=1, parameterName=WaitTime}))
  Multiply(A=Sine(None=Frac(None=@n60))) {constB=3}
 ]
}
