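# Translucent unlit Nexus pillar blend material that layers a glowing, panning mask pattern over an AO tint.
# UV1's vertical gradient shapes opacity and the pattern's falloff, while its inverse drives a second glow.
# Static switches control mask inversion and IsPillar2, which drops the second glow and vertex-colour masking (red for glow, green for pattern).
# An offset combines a push along vertex normals with a nudge toward the camera.
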Material /Figment_S08/Athena/Environments/Nexus/Materials/M_NexPillarBlend
graph {
 properties {blendMode=BLEND_Translucent, pixelDepthOffsetMode=PDOM_Legacy, shadingModel=MSM_Unlit, usedShadingModels=Unlit}
 nodes [
  n1 = VectorParameter() {parameterName=AOColor, defaultValue=<rgba 0.0002 0.0443 0.0781 1>}
  n2 = ComponentMask(None=TextureCoordinate() {coordinateIndex=1}) {g=true}
  n3 = OneMinus(None=@n2)
  n4 = ScalarParameter() {defaultValue=3.1562, parameterName=Power2}
  n5 = Multiply(
   A=Multiply(A=VectorParameter() {parameterName=Emisive2, defaultValue=<rgba 0 0.2634 0.3333 1>}.RGB, B=Power(Base=@n3, Exp=@n4))
   B=VertexColor().R
  )
  n6 = TextureCoordinate()
  n7 = ScalarParameter() {defaultValue=1.741, parameterName=Tiling}
  n8 = TextureSampleParameter2D(UVs=Panner(Coordinate=Multiply(A=@n6, B=@n7), Speed=ScalarParameter() {defaultValue=-0.02, parameterName=PanSpeed})) {parameterName=texture, samplerType=SAMPLERTYPE_Masks, texture=/Figment_S08/Athena/Environments/Nexus/Textures/NexFloor_2_M}
  n9 = StaticSwitchParameter(True=@n8.G, False=OneMinus(None=@n8.G)) {parameterName=InvertMask}
  n10 = ScalarParameter() {defaultValue=17.0228, parameterName=Power}
  n11 = Multiply(
   A=Power(Base=@n2, Exp=ScalarParameter() {defaultValue=6.7077, parameterName=Power3})
   B=Multiply(A=VectorParameter() {parameterName=Emissive, defaultValue=<rgba 5.0207 8.6904 10 1>}.RGB, B=Power(Base=@n9, Exp=@n10))
  )
  n12 = VertexColor()
  Add(
   A=StaticSwitchParameter(True=@n1.RGB, False=Add(A=@n1.RGB, B=@n5)) {parameterName=IsPillar2}
   B=StaticSwitchParameter(True=@n11, False=Multiply(A=@n11, B=@n12.G)) {parameterName=IsPillar2}
  )
  n13 = Subtract(A=CameraPositionWS(), B=WorldPosition().XYZ)
  Add(
   A=Multiply(A=VertexNormalWS(), B=ScalarParameter() {defaultValue=1, parameterName=Push})
   B=Multiply(A=Normalize(VectorInput=@n13), B=ScalarParameter() {defaultValue=1, parameterName=SurfaceOffset})
  )
  Multiply(A=@n2, B=ScalarParameter() {defaultValue=0.5524, parameterName=Opacity})
 ]
}
