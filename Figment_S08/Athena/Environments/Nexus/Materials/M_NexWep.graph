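# Weapon material for the Figment Nexus environment, combining diffuse, mask and normal maps with additive emissive, then passing the result through MF_NexPillarGlow in weapon mode.
# World Position Offset rotates vertices about an axis, then projects them along camera rays onto a local-space Min/Max box, scaled slightly against z-fighting.
# DisableVertexShader blends that projection back to plain rotation.
# A Time-driven sine chain is left unconnected.
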Material /Figment_S08/Athena/Environments/Nexus/Materials/M_NexWep
graph {
 properties {
  bUseMaterialAttributes=true
  bUsedWithMeshDeformer=true
  bUsedWithSkeletalMesh=true
  pixelDepthOffsetMode=PDOM_Legacy
  translucencyLightingMode=TLM_Surface
  usedShadingModels=Default Lit
 }
 nodes [
  n1 = TextureSampleParameter2D() {parameterName=Diffuse, texture=/Game/Weapons/FORT_Rifles/Textures/T_MachineGun_1_D}
  n2 = TextureSampleParameter2D() {parameterName=SpecularMask, samplerType=SAMPLERTYPE_Masks, texture=/Game/Weapons/FORT_Rifles/Textures/T_MachineGun_1_S}
  n3 = Add(
   A=Multiply(A=@n1.RGB, B=VectorParameter() {parameterName=EmissiveDiffuseAdd, defaultValue=<rgba 0.3281 0.3281 0.3281 1>}.RGB)
   B=Multiply(
    A=TextureSampleParameter2D() {parameterName=Emissive, samplerType=SAMPLERTYPE_Masks, texture=/Game/Packages/Fortress_SharedMaterials/Textures/Black_MASK}.RGB
    B=VectorParameter() {parameterName=EmissiveTextureColor, defaultValue=<rgba 0 0 0 1>}.RGB
   )
  )
  n4 = ScalarParameter() {parameterName=Opacity}
  n5 = TextureSampleParameter2D() {parameterName=Normal, samplerType=SAMPLERTYPE_Normal, texture=/Game/Weapons/FORT_Rifles/Textures/T_MachineGun_1_N}
  n6 = CameraPositionWS()
  n7 = TransformPosition(None=@n6) {transformSourceType=TRANSFORMPOSSOURCE_World}
  n8 = VectorParameter() {parameterName=RotationAxis, defaultValue=<rgba 1 0 0 1>}
  n9 = RotateAboutAxis(
   NormalizedRotationAxis=Normalize(VectorInput=Transform(None=@n8.RGB) {transformSourceType=TRANSFORMSOURCE_Local})
   RotationAngle=ScalarParameter() {parameterName=Rotation}
   PivotPoint=ObjectPositionWS()
   Position=WorldPosition() {worldPositionShaderOffset=WPT_ExcludeAllShaderOffsets}.XYZ
  )
  n10 = Add(A=WorldPosition() {worldPositionShaderOffset=WPT_ExcludeAllShaderOffsets}.XYZ, B=@n9)
  n11 = Normalize(VectorInput=Transform(None=Subtract(A=@n10, B=@n6)) {transformSourceType=TRANSFORMSOURCE_World, transformType=TRANSFORM_Local})
  n12 = VectorParameter() {parameterName=Min, defaultValue=<rgba -25 -800 -100 1>}
  n13 = VectorParameter() {parameterName=Max, defaultValue=<rgba 25 800 100 1>}
  n14 = Subtract(
   A=TransformPosition(None=MaterialFunctionCall!(Ray Origin=@n7, Ray Direction=@n11, Box Min=@n12.RGB, Box Max=@n13.RGB) {materialFunction=/Engine/Functions/Engine_MaterialFunctions01/Math/BoxIntersection}.Ray Entry Position) {transformType=TRANSFORMPOSSOURCE_World}
   B=WorldPosition() {worldPositionShaderOffset=WPT_ExcludeAllShaderOffsets}.XYZ
  )
  n15 = Add(A=@n14, B=Multiply(A=@n14, B=ScalarParameter() {defaultValue=0.01, parameterName=SelfZfightingFix}))
  n16 = ScalarParameter() {parameterName=DisableVertexShader}
  MaterialFunctionCall!(
   In=SetMaterialAttributes!(
    MaterialAttributes=MakeMaterialAttributes!(
     BaseColor=@n1.RGB
     Metallic=@n2.G
     Specular=@n2.G
     Roughness=@n2.G
     Anisotropy
     EmissiveColor=@n3
     Opacity=@n4
     OpacityMask
     Normal=@n5.RGB
     Tangent
     WorldPositionOffset
     SubsurfaceColor
     ClearCoat
     ClearCoatRoughness
     AmbientOcclusion
     Refraction
     CustomizedUVs_0
     CustomizedUVs_1
     CustomizedUVs_2
     CustomizedUVs_3
     CustomizedUVs_4
     CustomizedUVs_5
     CustomizedUVs_6
     CustomizedUVs_7
     PixelDepthOffset
     ShadingModel
     Displacement
    )
    World Position Offset=LinearInterpolate(A=@n15, B=@n9, Alpha=@n16)
   ) {attributeSetTypes [#1]}
   Centerlocation=Add(A=ActorPositionWS(), B=VectorParameter() {parameterName=Center, defaultValue=<rgba 0 0 299.434 1>}.RGB)
   Animation=Clamp(None=ScalarParameter() {parameterName=Animation})
   isWeapon=StaticBool() {value=true}
  ) {materialFunction=&/MF_NexPillarGlow}
  n17 = Multiply(A=Time(), B=ScalarParameter() {defaultValue=0.1, parameterName=TimeSpeed})
  n18 = Clamp(None=Multiply(A=Frac(None=@n17), B=ScalarParameter() {defaultValue=1, parameterName=WaitTime}))
  Multiply(A=Sine(None=Frac(None=@n18))) {constB=3}
 ]
}
