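# M_StarField_C is an additive, unlit, two-sided material that draws a glowing, tinted butterfly star field for the Nexus environment.
# It sums two red-channel samples of one texture: a layer animated through FlowMaps_Simple and a BumpOffset layer for fake depth.
# Color tints the sum, and inverted Fresnel fades grazing angles.
# Scaled camera-to-actor offset feeds MF_SpawnInBetween as world position offset, adding parallax movement.
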Material /Figment_S08/Athena/Environments/Nexus/Materials/M_StarField_C
graph {
 properties {blendMode=BLEND_Additive, pixelDepthOffsetMode=PDOM_Legacy, shadingModel=MSM_Unlit, twoSided=true, usedShadingModels=Unlit}
 nodes [
  n1 = CameraPositionWS()
  n2 = ActorPositionWS()
  MaterialFunctionCall!(WPO=Multiply(A=Subtract(A=@n1, B=@n2), B=ScalarParameter() {defaultValue=0.7, parameterName=Paralaxing})) {materialFunction=/Game/Athena/Environments/Nexus/Materials/MF_SpawnInBetween}
  n3 = AppendVector(A=Constant(), B=ScalarParameter() {defaultValue=0.025, parameterName=OffsetFlow})
  n4 = ScalarParameter() {defaultValue=0.2855, parameterName=Speed}
  n5 = Time()
  n6 = Multiply(
   A=TextureCoordinate()
   B=AppendVector(A=ScalarParameter() {defaultValue=8, parameterName=Tiling_X}, B=ScalarParameter() {defaultValue=1, parameterName=Tiling_Y})
  )
  n7 = ComponentMask(None=MaterialFunctionCall!(
    Diffuse=TextureObject() {texture=/Figment_S08/Athena/Environments/Nexus/Textures/T_ButterflyStarField_D}
    Normal Map
    "Flow Vector Map (see tooltip)"=@n3
    UV texture 2 offset=@n3
    "Time (see tooltip)"=Multiply(A=@n4, B=@n5)
    UVs=@n6
   ) {materialFunction=/Engine/Functions/Engine_MaterialFunctions02/Texturing/FlowMaps_Simple}.Diffuse) {r=true}
  n8 = BumpOffset(Coordinate=TextureCoordinate(), Height=ScalarParameter() {defaultValue=-1, parameterName=Height})
  n9 = AppendVector(
   A=ScalarParameter() {defaultValue=9.3937, parameterName=Tiling_X_Offset}
   B=ScalarParameter() {defaultValue=-0.2714, parameterName=Tiling_Y_Offset}
  )
  n10 = ScalarParameter() {defaultValue=9, parameterName=Tiling_X_1}
  n11 = ScalarParameter() {defaultValue=2.3077, parameterName=Tiling_Y_1}
  n12 = TextureSample(UVs=Multiply(A=Add(A=@n8, B=@n9), B=AppendVector(A=@n10, B=@n11))) {texture=&T_ButterflyStarField_D}
  n13 = ScalarParameter() {defaultValue=2.0768, parameterName=Power}
  Multiply(
   A=Multiply(A=VectorParameter() {parameterName=Color, defaultValue=<rgba 0.3039 1.2135 2 1>}.RGB, B=Add(A=@n7, B=@n12.R))
   B=OneMinus(None=Fresnel(ExponentIn=@n13) {baseReflectFraction=0})
  )
  ScalarParameter()
  TextureSample(UVs=@n6) {texture=&T_ButterflyStarField_D}
 ]
}
