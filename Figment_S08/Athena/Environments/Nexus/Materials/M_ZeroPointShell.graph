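# Unlit shell material for the Zero Point in the Nexus environment, showing a tinted, moving storm-cloud layer.
# Cloud mask samples virtual plane coordinates, distorted by scrolling noise, to blend two colors.
# A cosine of mesh V blends in a third color, then a tinted glow adds a V-centered band and radial distance falloff.
# MF_SpawnInBetween supplies World Position Offset.
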Material /Figment_S08/Athena/Environments/Nexus/Materials/M_ZeroPointShell
graph {
 properties {pixelDepthOffsetMode=PDOM_Legacy, shadingModel=MSM_Unlit, usedShadingModels=Unlit}
 nodes [
  n1 = Constant()
  n2 = MaterialFunctionCall(
   Plane Normal Axis=Constant3Vector() {constant=<rgba 0 0 1 0>}
   Plane X Axis=Constant3Vector() {constant=<rgba 1 0.0015 0 1>}
   Plane Center=MaterialFunctionCall(X=@n1, Y=@n1, Z=ScalarParameter() {defaultValue=7904700.5, parameterName=Planez}) {materialFunction=&MakeFloat3}.Result
   UV Scale=ScalarParameter() {defaultValue=48162788, parameterName=VirtualPlaneScale}
  ) {materialFunction=&VirtualPlaneCoordinates}
  n3 = Multiply(A=@n2.UVs, B=Constant2Vector() {g=-1, r=1})
  n4 = TextureCoordinate()
  n5 = Constant()
  n6 = ScalarParameter() {defaultValue=0.025, parameterName=SpeedDest}
  n7 = Frac(None=Multiply(A=Time(), B=AppendVector(A=@n5, B=@n6)))
  n8 = Multiply(
   A=TextureSample(UVs=Add(A=@n4, B=@n7)) {texture=/Figment_S08/Effects/Fort_Effects/Textures/Misc/T_Tiling_Noise_01}.R
   B=ScalarParameter() {defaultValue=0.6702, parameterName=Distort}
  )
  n9 = LinearInterpolate(
   A=VectorParameter() {parameterName=baseColor, defaultValue=<rgba 0.0834 0.093 0.1771 0>}.RGB
   B=VectorParameter() {parameterName=baseColor_1, defaultValue=<rgba 1.2 1.2 1.2 0>}.RGB
   Alpha=TextureSample(UVs=Add(A=@n3, B=@n8)) {samplerType=SAMPLERTYPE_Masks, texture=/Game/Athena/SafeZone/Textures/StormCloudtubes_M}.G
  )
  n10 = ScalarParameter() {defaultValue=0.4238, parameterName=Brightness}
  n11 = Cosine(None=ComponentMask(None=TextureCoordinate()) {g=true})
  n12 = TextureCoordinate()
  n13 = LinearInterpolate(
   A=ScalarParameter() {defaultValue=0.1048, parameterName=C_1}
   B=ScalarParameter() {defaultValue=94.2971, parameterName=C_2}
   Alpha=Distance(A=ComponentMask(None=@n12) {g=true}, B=Constant() {r=0.5})
  )
  n14 = Multiply(A=Clamp(None=OneMinus(None=@n13)), B=ScalarParameter() {defaultValue=0.1813, parameterName=FalloffBrightness_2})
  n15 = Divide(
   A=Distance(A=ComponentMask(None=@n2.World Position) {g=true, r=true}, B=Constant2Vector())
   B=ScalarParameter() {defaultValue=3387789824, parameterName=DistanceFadeBlendToWhite}
  )
  n16 = ScalarParameter() {defaultValue=142012528, parameterName=Max}
  n17 = Multiply(
   A=Power(Base=Clamp(None=@n15, Max=@n16), Exp=ScalarParameter() {defaultValue=0.0658, parameterName=FalloffExponent})
   B=ScalarParameter() {defaultValue=0.6312, parameterName=FalloffBrightness}
  )
  Add(
   A=LinearInterpolate(
    A=Multiply(A=@n9, B=@n10)
    B=VectorParameter() {parameterName=baseColor_2, defaultValue=<rgba 0.1636 0.2083 0.1979 0>}.RGB
    Alpha=Clamp(None=@n11)
   )
   B=Multiply(A=Add(A=@n14, B=@n17), B=VectorParameter() {parameterName=TintFalloff, defaultValue=<rgba 0.3594 0.5291 1 1>}.RGB)
  )
  MaterialFunctionCall!(WPO) {materialFunction=/Game/Athena/Environments/Nexus/Materials/MF_SpawnInBetween}
 ]
}
