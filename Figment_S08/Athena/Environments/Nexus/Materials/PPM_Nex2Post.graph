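# Post-process material for the Nexus environment that adds animated, tintable streak distortion to the scene in noisy, depth-weighted patches.
# Two counter-panning Perlin noise samples are summed, raised to a power and multiplied by normalized scene depth, so the mask favors distant pixels.
# A panning streak texture blends between two shift values to offset screen UVs, scaled by that mask.
# The mask then blends the tinted, distorted scene over the untouched image.
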Material /Figment_S08/Athena/Environments/Nexus/Materials/PPM_Nex2Post
graph {
 properties {materialDomain=MD_PostProcess, pixelDepthOffsetMode=PDOM_Legacy, usedShadingModels=Default Lit}
 nodes [
  n1 = TextureCoordinate()
  n2 = ScalarParameter() {defaultValue=0.3807, parameterName=TilingNoise}
  n3 = TextureSample(UVs=Panner(Coordinate=Multiply(A=@n1, B=@n2), Speed=ScalarParameter() {defaultValue=0.05, parameterName=SpeedPanNoise})) {texture=/Game/Global/Textures/Default/Noise/3xPerlin_M}
  n4 = TextureCoordinate()
  n5 = ScalarParameter() {defaultValue=0.7538, parameterName=TilingNoise2}
  n6 = TextureSample(UVs=Panner(Coordinate=Multiply(A=@n4, B=@n5), Speed=ScalarParameter() {defaultValue=-0.05, parameterName=SpeedPanNoise2})) {texture=&3xPerlin_M}
  n7 = ScalarParameter() {defaultValue=3.5281, parameterName=Power}
  n8 = Clamp(None=Power(Base=Add(A=@n3.R, B=@n6.R), Exp=@n7))
  n9 = SceneDepth()
  n10 = Clamp(None=Divide(A=ComponentMask(None=@n9.R) {r=true}, B=ScalarParameter() {defaultValue=693.7381}))
  Add(A=@n8, B=OneMinus(None=@n10))
  n11 = Multiply(A=TextureCoordinate(), B=ScalarParameter() {defaultValue=5.4852, parameterName=Tiling})
  n12 = ScalarParameter() {defaultValue=0.1, parameterName=SpeedPan}
  n13 = LinearInterpolate(
   A=ScalarParameter() {defaultValue=-0.03, parameterName=Shift1}
   B=ScalarParameter() {defaultValue=0.03, parameterName=Shift2}
   Alpha=TextureSample(UVs=Panner(Coordinate=@n11, Speed=@n12)) {samplerType=SAMPLERTYPE_Masks, texture=/Figment_S08/Athena/Environments/Nexus/Textures/T_StreaksH}.G
  )
  n14 = Saturate(None=Multiply(A=@n8, B=@n10))
  n15 = SceneTexture(UVs=Add(A=TextureCoordinate(), B=Multiply(A=@n13, B=@n14))) {sceneTextureId=PPI_PostProcessInput0}
  LinearInterpolate(
   A=Multiply(A=ComponentMask(None=@n15.Color) {b=true, g=true, r=true}, B=VectorParameter() {parameterName=Tint, defaultValue=<rgba 1 1 1 1>}.RGB)
   B=ComponentMask(None=SceneTexture() {sceneTextureId=PPI_PostProcessInput0}.Color) {b=true, g=true, r=true}
   Alpha=OneMinus(None=@n14)
  )
  LinearInterpolate(A=@n7, B=ScalarParameter() {defaultValue=57.2498, parameterName=Power2}, Alpha=Multiply(A=@n3.R, B=@n6.R))
  Multiply()
  TextureSample() {texture=&3xPerlin_M}
 ]
}
